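# Pre-simulation Control Rig for the WormChalk male character body, posing BASE bones that parent its rag and sweatshirt bones.
# After the shared DeformRig function, a sequence parent-constrains left-arm rag bones and offsets them via spherical pose readers.
# Thigh pose readers and spine bend drive sweatshirt front and side offsets, then aim constraints point front bones at the calves.
# An interpolated skydive_dive_mode weight rotates sweatshirt bones for diving.
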ControlRigBlueprint /BRCosmetics/Characters/Player/Male/Medium/Bodies/M_MED_WormChalk/Meshes/M_MED_WormChalk_CtrlRig_presim
graphs [
 RigVMFunctionLibrary {links [], nodes []}
 RigVMFunction_Sequence_1_ContainedGraph {nodes [
   n1 = RigVMFunctionEntryNode()
   RigVMFunction_Sequence = &RigVMFunction_Sequence(ExecuteContext=@n1.ExecuteContext)
   n2 = &RigVMFunction_Sequence(ExecuteContext=@RigVMFunction_Sequence.B)
   n3 = &RigVMFunction_Sequence(ExecuteContext=@n2.B)
   n4 = &RigVMFunction_Sequence(ExecuteContext=@n3.B)
   n5 = &RigVMFunction_Sequence(ExecuteContext=@n4.B)
   RigVMFunctionReturnNode(A=@RigVMFunction_Sequence.A, B=@n2.A, C=@n3.A, D=@n4.A, E=@n5.A, F=@n5.B)
  ]}
 RigVMModel {nodes [
   RigUnit_BeginExecution = &RigUnit_BeginExecution()
   FN_DeformRig_FNC = RigVMFunctionReferenceNode(
    DeformRig_ Curve="(deformRig_l_upperarm_up_alpha_72_#1=False,deformRig_r_upperarm_up_alpha_73_#2=False,deformRig_l_upperarm_fr_alpha_74_#3=False,deformRig_r_upperarm_fr_alpha_75_CB51B8 ...(truncated)"
    DeformRig_Bool="(upperarm_1_#4=True,elbow_3_#5=True,wrist_7_#6=True,groin_8_#7=True,glute_17_#8=True,knee_9_40C408324CA261 ...(truncated)"
    DeformRig_Debug=(upperarm_debug_18_#4=False,wrist_debug_20_#6=False,lat_debug_26_#9=False,groinGlute_debug_27_#7=False)
    DeformRig_Value="(upperarm_up_18_#10=(Rotation=(X=0,Y=0,Z=0,W=1),Translation=(X=0,Y=0,Z=8),Scale3D=(X=0.7,Y=1,Z=1)),upperarm_fwd_fr_19_#11=(Rotation=(X=0,Y=0,Z=0,W=1),Translation=(X=0,Y=10,Z=0),Scale3D=(X=0.5,Y=1,Z= ...(truncated)"
    ExecuteContext=@RigUnit_BeginExecution.ExecutePin
   )
   > RigVMFunction_Sequence = RigVMAggregateNode()
   n1 = &RigUnit_SphericalPoseReader(
    ActiveRegionScaleFactors=(PositiveWidth=1,NegativeWidth=0.1,PositiveHeight=1,NegativeHeight=0.5)
    ActiveRegionSize=0.3
    Debug=(bDrawDebug=False,bDraw2D=false,bDrawLocalAxes=false,DebugScale=30,DebugSegments=30,DebugThickness=0.2)
    DriverAxis=(X=1,Y=0,Z=0)
    DriverItem=(Type=Bone,Name="upperarm_l")
    FalloffRegionScaleFactors=(PositiveWidth=1,NegativeWidth=0.5,PositiveHeight=1,NegativeHeight=1)
    FalloffSize=0.5
    OptionalParentItem=(Type=Bone,Name="pelvis")
    RotationOffset=(X=120,Y=200,Z=0)
    ExecutePin=@RigVMFunction_Sequence.B
   )
   > &RigUnit_ModifyTransforms(
    ItemToModify="((Item=(Type=Bone,Name=\"BASE_rag_top_1\"),Transform=(Rotation=(X=0.0872,Y=0,Z=0,W=0.9962),Translation=(X=-3,Y=1,Z=0),Scale3D=(X=1,Y=1,Z=1))),(Item=(Type=Bone,Name=\"BASE_rag_top_2\"),Transform=(Rotation=(X=0,Y=0,Z=0,W=1),Translation=(X=-3,Y=0,Z=1),Scale3D=(X= ...(truncated)"
    Weight=@n1.OutputParam
   )
   > &RigUnit_ModifyTransforms(
    ItemToModify="((Item=(Type=Bone,Name=\"BASE_rag_bot_1\"),Transform=(Rotation=(X=0,Y=0,Z=0,W=1),Translation=(X=-4,Y=0,Z=0),Scale3D=(X=1,Y=1,Z=1))),(Item=(Type=Bone,Name=\"BASE_rag_bot_2\"),Transform=(Rotation=(X=0,Y=0,Z=0,W=1),Translation=(X=-4,Y=0,Z=1.5),Scale3D=(X=1,Y=1,Z= ...(truncated)"
    Weight=@n1.OutputParam
   )
   > &RigUnit_SphericalPoseReader(
    ActiveRegionScaleFactors=(PositiveWidth=1,NegativeWidth=0.1,PositiveHeight=1,NegativeHeight=0.5)
    ActiveRegionSize=0.3
    Debug=(bDrawDebug=False,bDraw2D=false,bDrawLocalAxes=false,DebugScale=30,DebugSegments=30,DebugThickness=0.2)
    DriverAxis=(X=1,Y=0,Z=0)
    DriverItem=(Type=Bone,Name="upperarm_l")
    FalloffRegionScaleFactors=(PositiveWidth=1,NegativeWidth=0.5,PositiveHeight=1,NegativeHeight=1)
    FalloffSize=0.5
    OptionalParentItem=(Type=Bone,Name="pelvis")
    RotationOffset=(X=180,Y=0,Z=0)
   )
   > &RigUnit_ModifyTransforms(
    ItemToModify="((Item=(Type=Bone,Name=\"BASE_rag_bot_1\"),Transform=(Rotation=(X=0,Y=0,Z=0,W=1),Translation=(X=0,Y=0,Z=0),Scale3D=(X=1,Y=1,Z=1))),(Item=(Type=Bone,Name=\"BASE_rag_bot_2\"),Transform=(Rotation=(X=0,Y=0,Z=0,W=1),Translation=(X=0,Y=0,Z=0),Scale3D=(X=1,Y=1,Z=1))) ...(truncated)"
    Weight=0
   )
   n2 = &RigUnit_SphericalPoseReader(
    ActiveRegionScaleFactors=(PositiveWidth=0.7,NegativeWidth=1,PositiveHeight=1,NegativeHeight=1)
    ActiveRegionSize=0.25
    Debug=(bDrawDebug=True,bDraw2D=false,bDrawLocalAxes=false,DebugScale=25,DebugSegments=20,DebugThickness=0.25)
    DriverAxis=(X=-1,Y=0,Z=0)
    DriverItem=(Type=Bone,Name="thigh_l")
    FalloffRegionScaleFactors=(PositiveWidth=0.2,NegativeWidth=0.2,PositiveHeight=0.2,NegativeHeight=0.2)
    FalloffSize=1
    FlipWidthScaling=False
    OptionalParentItem=(Type=Bone)
    RotationOffset=(X=120,Y=10,Z=0)
    ExecutePin=@RigVMFunction_Sequence.E
   )
   > &RigUnit_ModifyTransforms(
    ItemToModify="((Item=(Type=Bone,Name=\"BASE_sweatshirt_L_front_03\"),Transform=(Rotation=(X=0,Y=0,Z=0,W=1),Translation=(X=-5,Y=0,Z=0),Scale3D=(X=1,Y=1,Z=1))),(Item=(Type=Bone,Name=\"BASE_sweatshirt_L_side_03_F\"),Transform=(Rotation=(X=0,Y=0,Z=0,W=1),Translation=(X=0,Y=0,Z= ...(truncated)"
    ItemToModify.0.Transform.Rotation=&RigVMFunction_MathQuaternionFromEuler(Euler=(X=0,Y=-90,Z=-40), RotationOrder=XZY).Result
    ItemToModify.1.Transform.Rotation=&RigVMFunction_MathQuaternionFromEuler(Euler=(X=0,Y=-30,Z=0), RotationOrder=XZY).Result
    Weight=@n2.OutputParam
   )
   n3 = &RigUnit_GetRelativeTransformForItem(Child=(Type=Bone,Name="spine_05"), Parent=(Type=Bone,Name="spine_01"), bChildInitial=False, bParentInitial=False)
   > &RigUnit_ModifyTransforms(
    ItemToModify="((Item=(Type=Bone,Name=\"BASE_sweatshirt_L_front_03\"),Transform=(Rotation=(X=0,Y=0,Z=0,W=1),Translation=(X=-5,Y=0,Z=0),Scale3D=(X=1,Y=1,Z=1))),(Item=(Type=Bone,Name=\"BASE_sweatshirt_L_side_03_F\"),Transform=(Rotation=(X=0,Y=0,Z=0,W=1),Translation=(X=0,Y=0,Z= ...(truncated)"
    ItemToModify.1.Transform.Rotation=&RigVMFunction_MathQuaternionFromEuler(Euler=(X=0,Y=-30,Z=0), RotationOrder=XZY).Result
    ItemToModify.0.Transform.Rotation=&RigVMFunction_MathQuaternionFromEuler(Euler=(X=0,Y=-90,Z=-40), RotationOrder=XZY).Result
    Weight=&RigVMFunction_MathDoubleRemap(
     SourceMaximum=20
     SourceMinimum=10
     TargetMaximum=1
     TargetMinimum=0
     Value=&RigVMFunction_MathQuaternionToEuler(RotationOrder=ZYX, Value=@n3.RelativeTransform.Rotation).Result.Z
    ).Result
   )
   > n4 = &RigUnit_SphericalPoseReader(
    ActiveRegionScaleFactors=(PositiveWidth=0.7,NegativeWidth=1,PositiveHeight=1,NegativeHeight=1)
    ActiveRegionSize=0.1
    Debug=(bDrawDebug=True,bDraw2D=false,bDrawLocalAxes=false,DebugScale=25,DebugSegments=20,DebugThickness=0.25)
    DriverAxis=(X=1,Y=0,Z=0)
    DriverItem=(Type=Bone,Name="thigh_r")
    FalloffRegionScaleFactors=(PositiveWidth=0.2,NegativeWidth=0.2,PositiveHeight=0.2,NegativeHeight=0.2)
    FalloffSize=1
    FlipWidthScaling=False
    OptionalParentItem=(Type=Bone)
    RotationOffset=(X=-60,Y=40,Z=0)
   )
   > &RigUnit_ModifyTransforms(
    ItemToModify=((Item=(Type=Bone,Name="BASE_sweatshirt_R_front_03"),Transform=(Rotation=(X=0,Y=0,Z=0,W=1),Translation=(X=-7,Y=-2,Z=0),Scale3D=(X=1,Y=1,Z=1))))
    ItemToModify.0.Transform.Rotation=&RigVMFunction_MathQuaternionFromEuler(Euler=(X=0,Y=-90,Z=-40), RotationOrder=XZY).Result
    Weight=@n4.OutputParam
   )
   &RigUnit_AimConstraintLocalSpaceOffset(
    AdvancedSettings=(DebugSettings=(bEnabled=true,Scale=10,WorldOffset=(Rotation=(X=0,Y=0,Z=0,W=1),Translation=(X=0,Y=0,Z=0),Scale3D=(X=1,Y=1,Z=1))),RotationOrderForFilter=XZY)
    AimAxis=(X=-1,Y=0,Z=0)
    Child=(Type=Bone,Name="BASE_sweatshirt_L_front_03")
    Filter=(bX=True,bY=True,bZ=True)
    Parents=((Item=(Type=Bone,Name="calf_l"),Weight=1))
    UpAxis=(X=0,Y=1,Z=0)
    Weight=1
    WorldUp=(Target=(X=0,Y=0,Z=1),Kind=Location,Space=(Type=Bone,Name="calf_l"))
    ExecutePin=@RigVMFunction_Sequence.F
   )
   > &RigUnit_AimConstraintLocalSpaceOffset(
    AdvancedSettings=(DebugSettings=(bEnabled=true,Scale=10,WorldOffset=(Rotation=(X=0,Y=0,Z=0,W=1),Translation=(X=0,Y=0,Z=0),Scale3D=(X=1,Y=1,Z=1))),RotationOrderForFilter=XZY)
    AimAxis=(X=-1,Y=0,Z=0)
    Child=(Type=Bone,Name="BASE_sweatshirt_R_front_03")
    Filter=(bX=True,bY=True,bZ=True)
    Parents=((Item=(Type=Bone,Name="calf_r"),Weight=1))
    UpAxis=(X=0,Y=1,Z=0)
    Weight=1
    WorldUp=(Target=(X=0,Y=0,Z=1),Kind=Location,Space=(Type=Bone,Name="calf_r"))
   )
   RigVMVariableNode(Value=1, ExecuteContext=@RigVMFunction_Sequence.A) {variable=togle_01}
   n5 = RigVMVariableNode() {variable=togle_01}
   > &RigUnit_ParentConstraint(
    AdvancedSettings=(InterpolationType=Average,RotationOrderForFilter=XZY)
    Child=(Type=Bone,Name="BASE_rag_bot_1")
    Parents="((Item=(Type=Bone,Name=\"spine_03\"),Weight=0.0027),(Item=(Type=Bone,Name=\"upperarm_twist_01_l\"),Weight=0.7744),(Item=(Type=Bone,Name=\"upperarm_l\"),Weight=0.0823),(Item=(Type=Bone,Name=\"upperarm_twist_02_l\"),Weight=0.0943),(Item=(Type=Bone,Name=\"clavicle_l\") ...(truncated)"
    Weight=@n5.Value
   )
   > &RigUnit_ParentConstraint(
    AdvancedSettings=(InterpolationType=Average,RotationOrderForFilter=XZY)
    Child=(Type=Bone,Name="BASE_rag_bot_2")
    Parents=((Item=(Type=Bone,Name="clavicle_l"),Weight=0.0503),(Item=(Type=Bone,Name="upperarm_twist_01_l"),Weight=0.4929),(Item=(Type=Bone,Name="upperarm_l"),Weight=0.1971),(Item=(Type=Bone,Name="upperarm_twist_02_l"),Weight=0.2597))
    Weight=@n5.Value
   )
   > &RigUnit_ParentConstraint(
    AdvancedSettings=(InterpolationType=Average,RotationOrderForFilter=XZY)
    Child=(Type=Bone,Name="BASE_rag_bot_3")
    Parents="((Item=(Type=Bone,Name=\"spine_03\"),Weight=0.0324),(Item=(Type=Bone,Name=\"upperarm_twist_01_l\"),Weight=0.5884),(Item=(Type=Bone,Name=\"upperarm_l\"),Weight=0.1408),(Item=(Type=Bone,Name=\"upperarm_twist_02_l\"),Weight=0.1593),(Item=(Type=Bone,Name=\"clavicle_l\") ...(truncated)"
    Weight=@n5.Value
   )
   > RigVMVariableNode(Value=1) {variable=togle_02}
   > &RigUnit_ParentConstraint(
    AdvancedSettings=(InterpolationType=Average,RotationOrderForFilter=XZY)
    Child=(Type=Bone,Name="BASE_rag_top_1")
    Parents="((Item=(Type=Bone,Name=\"spine_03\"),Weight=0.0287),(Item=(Type=Bone,Name=\"upperarm_twist_01_l\"),Weight=0.6996),(Item=(Type=Bone,Name=\"upperarm_l\"),Weight=0.0003),(Item=(Type=Bone,Name=\"upperarm_twist_02_l\"),Weight=0),(Item=(Type=Bone,Name=\"clavicle_l\"),Weig ...(truncated)"
    Weight=0
    bMaintainOffset=true
   )
   n6 = RigVMVariableNode() {variable=togle_02}
   > &RigUnit_ParentConstraint(
    AdvancedSettings=(InterpolationType=Average,RotationOrderForFilter=XZY)
    Child=(Type=Bone,Name="BASE_rag_top_2")
    Parents="((Item=(Type=Bone,Name=\"clavicle_l\"),Weight=0.1948),(Item=(Type=Bone,Name=\"upperarm_twist_01_l\"),Weight=0.744),(Item=(Type=Bone,Name=\"upperarm_l\"),Weight=0.0197),(Item=(Type=Bone,Name=\"upperarm_twist_02_l\"),Weight=0.0187),(Item=(Type=Bone,Name=\"spine_03\"), ...(truncated)"
    bMaintainOffset=true
    Weight=@n6.Value
   )
   > &RigUnit_ParentConstraint(
    AdvancedSettings=(InterpolationType=Average,RotationOrderForFilter=XZY)
    Child=(Type=Bone,Name="BASE_rag_top_3")
    Parents="((Item=(Type=Bone,Name=\"spine_03\"),Weight=0.0954),(Item=(Type=Bone,Name=\"upperarm_twist_01_l\"),Weight=0.3538),(Item=(Type=Bone,Name=\"upperarm_l\"),Weight=0.0001),(Item=(Type=Bone,Name=\"rag_top_2\"),Weight=0.0001),(Item=(Type=Bone,Name=\"clavicle_l\"),Weight=0. ...(truncated)"
    Weight=@n6.Value
   )
   &RigUnit_ModifyTransforms(
    ItemToModify="((Item=(Type=Bone,Name=\"extra_spine_05\"),Transform=(Rotation=(X=0,Y=0,Z=0,W=1),Translation=(X=0,Y=0,Z=0),Scale3D=(X=0.5,Y=0.5,Z=0.5))),(Item=(Type=Bone,Name=\"BASE_sweatshirt_R_side_03\"),Transform=(Rotation=(X=0,Y=0,Z=0,W=1),Translation=(X=0,Y=0,Z=1),Scale3 ...(truncated)"
    Weight=1
    ExecutePin=@RigVMFunction_Sequence.C
   )
   n7 = &RigVMFunction_AlphaInterp(InterpSpeedDecreasing=5, InterpSpeedIncreasing=5, Scale=1, bClampResult=False, Value=RigVMVariableNode() {variable=skydive_dive_mode}.Value)
   RigVMFunction_MathQuaternionFromEuler = &RigVMFunction_MathQuaternionFromEuler(Euler=(X=0,Y=60,Z=0), RotationOrder=XYZ)
   > &RigUnit_ModifyTransforms(
    ItemToModify="((Item=(Type=Bone,Name=\"BASE_sweatshirt_R_back\"),Transform=(Rotation=(X=0,Y=0.5,Z=0,W=0.866),Translation=(X=0,Y=0,Z=-1),Scale3D=(X=1,Y=1,Z=1))),(Item=(Type=Bone,Name=\"BASE_sweatshirt_M_back\"),Transform=(Rotation=(X=0,Y=0.5,Z=0,W=0.866),Translation=(X=0,Y=0 ...(truncated)"
    Weight=@n7.Result
    ItemToModify.4.Transform.Rotation=@RigVMFunction_MathQuaternionFromEuler.Result
    ItemToModify.0.Transform.Rotation=&RigVMFunction_MathQuaternionFromEuler(Euler=(X=0,Y=60,Z=0), RotationOrder=XYZ).Result
    ItemToModify.1.Transform.Rotation=&RigVMFunction_MathQuaternionFromEuler(Euler=(X=0,Y=60,Z=0), RotationOrder=XYZ).Result
    ItemToModify.2.Transform.Rotation=&RigVMFunction_MathQuaternionFromEuler(Euler=(X=0,Y=60,Z=0), RotationOrder=XYZ).Result
    ItemToModify.3.Transform.Rotation=&RigVMFunction_MathQuaternionFromEuler(Euler=(X=0,Y=60,Z=0), RotationOrder=XYZ).Result
    ItemToModify.5.Transform.Rotation=&RigVMFunction_MathQuaternionFromEuler(Euler=(X=0,Y=60,Z=0), RotationOrder=XYZ).Result
   )
  ]}
]
hierarchy [
 root {loc=<0 0 0>, rot=<0 0 0 1>}
 pelvis {parent=root, loc=<0 -0.7071 93.1191>, rot=<0.0127 0.707 -0.0127 -0.707>}
 spine_01 {parent=pelvis, loc=<6.7494 -0.6388 0>, rot=<0 0 0.034 -0.9994>}
 spine_02 {parent=spine_01, loc=<7.252 -0.035 0>, rot=<0 0 0.0238 -0.9997>}
 spine_03 {parent=spine_02, loc=<6.0106 0 0>, rot=<0 0 0.0149 -0.9999>}
 spine_04 {parent=spine_03, loc=<8.1023 0 0>, rot=<0 0 -0.0132 -0.9999>}
 spine_05 {parent=spine_04, loc=<10.0142 0 0>, rot=<0 0 -0.0015 -1>}
 attach_backpack {parent=spine_05, loc=<-0.0457 -13.2495 -0.0212>, rot=<-0.0092 -0.7071 0.0093 -0.707>}
 neck_01 {parent=spine_05, loc=<11.3138 0.1529 0>, rot=<0 0 -0.2003 -0.9797>}
 neck_02 {parent=neck_01, loc=<4.5627 0 0>, rot=<0 0 0.0504 -0.9987>}
 head {parent=neck_02, loc=<5.3914 0 0>, rot=<0 0 0.115 -0.9934>}
 BASE_M_hair {parent=head, loc=<9.5468 -7.5093 0>, rot=<0 0 0.9951 -0.0993>}
 dyn_M_hair_01 {parent=BASE_M_hair, loc=<0 0 0>, rot=<0 0 0 1>}
 dyn_M_hair_02 {parent=dyn_M_hair_01, loc=<8.7791 0 0>, rot=<0 0 -0.0991 -0.9951>}
 dyn_M_hair_03 {parent=dyn_M_hair_02, loc=<8.2692 0 0>, rot=<0 0 0.0059 -1>}
 dyn_M_hair_04 {parent=dyn_M_hair_03, loc=<8.7845 0 0>, rot=<0 0 0.1752 -0.9845>}
 BASE_L_hair {parent=head, loc=<9.5646 -5.0822 -5.5493>, rot=<-0.0721 -0.016 0.9964 -0.0413>}
 dyn_L_hair_01 {parent=BASE_L_hair, loc=<0 0 0>, rot=<0 0 0 1>}
 dyn_L_hair_02 {parent=dyn_L_hair_01, loc=<8.7791 0 0>, rot=<0 0 -0.1979 -0.9802>}
 dyn_L_hair_03 {parent=dyn_L_hair_02, loc=<8.2692 0 0>, rot=<0 0 -0.0305 -0.9995>}
 dyn_L_hair_04 {parent=dyn_L_hair_03, loc=<8.7845 0 0>, rot=<0 0 0.201 -0.9796>}
 clavicle_l {parent=spine_05, loc=<4.3575 -0.1466 -2.3223>, rot=<0.0465 -0.7289 -0.009 -0.6829>}
 upperarm_l {parent=clavicle_l, loc=<15.5111 0 0>, rot=<0.0226 -0.3234 -0.0088 -0.9459>}
 lowerarm_l {parent=upperarm_l, loc=<26.6514 0 0>, rot=<0.0391 -0.0099 -0.2453 -0.9686>}
 hand_l {parent=lowerarm_l, loc=<22.3491 0 0>, rot=<-0.7081 -0.0507 0.0975 -0.6975>}
 thumb_01_l {parent=hand_l, loc=<1.3118 -0.3191 -2.459>, rot=<0.6546 -0.4042 0.0664 -0.6354>}
 thumb_02_l {parent=thumb_01_l, loc=<5.9336 -0.0448 0.0401>, rot=<-0.0037 0.0213 0.298 -0.9543>}
 thumb_03_l {parent=thumb_02_l, loc=<2.2223 0.3682 -0.0907>, rot=<0.0118 -0.013 0.0066 -0.9998>}
 index_metacarpal_l {parent=hand_l, loc=<1.6535 0.0183 -2.96>, rot=<0.0158 -0.0124 -0.0173 -0.9996>}
 index_01_l {parent=index_metacarpal_l, loc=<8.3008 0.0357 -0.0524>, rot=<-0.0044 0.0485 0.0924 -0.9945>}
 index_02_l {parent=index_01_l, loc=<3.3587 -0.3004 0.0452>, rot=<-0.002 0.0027 0.3225 -0.9466>}
 index_03_l {parent=index_02_l, loc=<2.5578 -0.123 -0.0252>, rot=<-0.002 0.0067 0.2558 -0.9667>}
 middle_metacarpal_l {parent=hand_l, loc=<2.1856 -0.1725 -0.5984>, rot=<0.006 0.0168 -0.0329 -0.9993>}
 middle_01_l {parent=middle_metacarpal_l, loc=<7.3275 -0.2447 0.0129>, rot=<-0.0208 0.0042 0.0858 -0.9961>}
 middle_02_l {parent=middle_01_l, loc=<4.0123 0.1401 0.1175>, rot=<-0.0218 0.0258 0.3512 -0.9357>}
 middle_03_l {parent=middle_02_l, loc=<2.6077 0.0534 0.1373>, rot=<0.0082 0.0188 0.1708 -0.9851>}
 ring_metacarpal_l {parent=hand_l, loc=<2.2074 -0.3625 1.3638>, rot=<-0.0284 0.0722 -0.0458 -0.9959>}
 ring_01_l {parent=ring_metacarpal_l, loc=<7.0295 -0.6197 -0.1446>, rot=<-0.0489 -0.0252 0.1892 -0.9804>}
 ring_02_l {parent=ring_01_l, loc=<3.6174 0.4647 -0.1027>, rot=<0 0 0.2806 -0.9598>}
 ring_03_l {parent=ring_02_l, loc=<2.4152 0.114 0.0199>, rot=<0 0 0.1189 -0.9929>}
 pinky_metacarpal_l {parent=hand_l, loc=<1.7616 -0.9753 3.3814>, rot=<-0.0397 0.0814 -0.069 -0.9935>}
 pinky_01_l {parent=pinky_metacarpal_l, loc=<7.1709 -0.4831 -0.1094>, rot=<-0.0654 -0.0035 0.2242 -0.9723>}
 pinky_02_l {parent=pinky_01_l, loc=<2.8274 0.0893 -0.1043>, rot=<0.0037 0.0011 0.2854 -0.9584>}
 pinky_03_l {parent=pinky_02_l, loc=<2.2046 0.0307 -0.0078>, rot=<-0.0002 0.0098 0.1445 -0.9895>}
 lowerarm_twist_02_l {parent=lowerarm_l, loc=<9.6199 0 0>, rot=<0 0 0 -1>}
 lowerarm_twist_01_l {parent=lowerarm_l, loc=<17.8549 0 0>, rot=<0 0 0 -1>}
 deform_elbow_in_base_l {parent=lowerarm_l, loc=<0.2112 5.259 0.8274>, rot=<0 0 0 -1>}
 deform_elbow_in_l {parent=deform_elbow_in_base_l, loc=<0 0 0>, rot=<0 0 0 1>}
 deform_elbow_out_base_l {parent=lowerarm_l, loc=<0 -4 0>, rot=<0 0 0 -1>}
 deform_elbow_out_l {parent=deform_elbow_out_base_l, loc=<0 0 0>, rot=<0 0 0 1>}
 upperarm_twist_01_l {parent=upperarm_l, loc=<0.0041 0 0>, rot=<0 0 0 -1>}
 BASE_rag_top_1 {parent=upperarm_twist_01_l, loc=<1.4305 6.5988 -0.0633>, rot=<-0.2035 -0.0776 0.0269 -0.9756>}
 rag_top_1 {parent=BASE_rag_top_1, loc=<0 0 0>, rot=<0 0 0 1>}
 BASE_rag_top_2 {parent=upperarm_twist_01_l, loc=<4.5602 -1.4392 5.0796>, rot=<-0.0065 -0.0543 -0.0182 -0.9983>}
 rag_top_2 {parent=BASE_rag_top_2, loc=<0 0 0>, rot=<0 0 0 1>}
 BASE_rag_top_3 {parent=upperarm_twist_01_l, loc=<0.7822 -8.2275 -1.4886>, rot=<0.0466 -0.1155 -0.0408 -0.9914>}
 rag_top_3 {parent=BASE_rag_top_3, loc=<0 0 0>, rot=<0 0 0 1>}
 upperarm_twist_02_l {parent=upperarm_l, loc=<13.7941 0 0>, rot=<0 0 0 -1>}
 BASE_rag_bot_2 {parent=upperarm_twist_02_l, loc=<-1.7008 -0.8892 4.0314>, rot=<-0.0073 -0.0117 -0.0179 -0.9997>}
 rag_bot_2 {parent=BASE_rag_bot_2, loc=<0 0 0>, rot=<0 0 0 1>}
 BASE_rag_bot_1 {parent=upperarm_twist_02_l, loc=<-3.8882 7.0574 -1.6563>, rot=<-0.2035 -0.0776 0.0269 -0.9756>}
 rag_bot_1 {parent=BASE_rag_bot_1, loc=<0 0 0>, rot=<0 0 0 1>}
 BASE_rag_bot_3 {parent=upperarm_twist_02_l, loc=<-5.2522 -8.1298 -3.9627>, rot=<-0.2035 -0.0776 0.0269 -0.9756>}
 rag_bot_3 {parent=BASE_rag_bot_3, loc=<0 0 0>, rot=<0 0 0 1>}
 attach_cape {parent=spine_05, loc=<9.0264 -11.5526 -0.0884>, rot=<0.0617 -0.7044 -0.0616 -0.7044>}
 clavicle_r {parent=spine_05, loc=<4.3575 -0.1466 2.3223>, rot=<0.7289 0.0465 -0.6829 0.009>}
 upperarm_r {parent=clavicle_r, loc=<-15.5111 0 0>, rot=<0.0226 -0.3234 -0.0088 -0.9459>}
 lowerarm_r {parent=upperarm_r, loc=<-26.6514 0 0>, rot=<0.0391 -0.0099 -0.2453 -0.9686>}
 hand_r {parent=lowerarm_r, loc=<-22.3491 0 0>, rot=<-0.7081 -0.0507 0.0975 -0.6975>}
 thumb_01_r {parent=hand_r, loc=<-1.3118 0.3191 2.459>, rot=<0.6546 -0.4042 0.0664 -0.6354>}
 thumb_02_r {parent=thumb_01_r, loc=<-5.9336 0.0448 -0.0401>, rot=<-0.0037 0.0213 0.298 -0.9543>}
 thumb_03_r {parent=thumb_02_r, loc=<-2.2223 -0.3682 0.0907>, rot=<0.0118 -0.013 0.0066 -0.9998>}
 index_metacarpal_r {parent=hand_r, loc=<-1.6535 -0.0183 2.96>, rot=<0.0158 -0.0124 -0.0173 -0.9996>}
 index_01_r {parent=index_metacarpal_r, loc=<-8.3008 -0.0357 0.0524>, rot=<-0.0044 0.0485 0.0924 -0.9945>}
 index_02_r {parent=index_01_r, loc=<-3.3587 0.3004 -0.0452>, rot=<-0.002 0.0027 0.3225 -0.9466>}
 index_03_r {parent=index_02_r, loc=<-2.5578 0.123 0.0252>, rot=<-0.002 0.0067 0.2558 -0.9667>}
 middle_metacarpal_r {parent=hand_r, loc=<-2.1856 0.1725 0.5984>, rot=<0.006 0.0168 -0.0329 -0.9993>}
 middle_01_r {parent=middle_metacarpal_r, loc=<-7.3275 0.2447 -0.0129>, rot=<-0.0208 0.0042 0.0858 -0.9961>}
 middle_02_r {parent=middle_01_r, loc=<-4.0123 -0.1401 -0.1175>, rot=<-0.0218 0.0258 0.3512 -0.9357>}
 middle_03_r {parent=middle_02_r, loc=<-2.6077 -0.0534 -0.1373>, rot=<0.0082 0.0188 0.1708 -0.9851>}
 ring_metacarpal_r {parent=hand_r, loc=<-2.2074 0.3625 -1.3638>, rot=<-0.0284 0.0722 -0.0458 -0.9959>}
 ring_01_r {parent=ring_metacarpal_r, loc=<-7.0295 0.6197 0.1446>, rot=<-0.0489 -0.0252 0.1892 -0.9804>}
 ring_02_r {parent=ring_01_r, loc=<-3.6174 -0.4647 0.1027>, rot=<0 0 0.2806 -0.9598>}
 ring_03_r {parent=ring_02_r, loc=<-2.4152 -0.114 -0.0199>, rot=<0 0 0.1189 -0.9929>}
 pinky_metacarpal_r {parent=hand_r, loc=<-1.7616 0.9753 -3.3814>, rot=<-0.0397 0.0814 -0.069 -0.9935>}
 pinky_01_r {parent=pinky_metacarpal_r, loc=<-7.1709 0.4831 0.1094>, rot=<-0.0654 -0.0035 0.2242 -0.9723>}
 pinky_02_r {parent=pinky_01_r, loc=<-2.8274 -0.0893 0.1043>, rot=<0.0037 0.0011 0.2854 -0.9584>}
 pinky_03_r {parent=pinky_02_r, loc=<-2.2046 -0.0307 0.0078>, rot=<-0.0002 0.0098 0.1445 -0.9895>}
 lowerarm_twist_02_r {parent=lowerarm_r, loc=<-9.6199 0 0>, rot=<0 0 0 1>}
 lowerarm_twist_01_r {parent=lowerarm_r, loc=<-17.8549 0 0>, rot=<0 0 0 -1>}
 BASE_sleeve_01 {parent=lowerarm_r, loc=<2.083 -0.929 -5.2867>, rot=<-0.0527 -0.993 -0.0293 -0.1011>}
 dyn_sleeve_01 {parent=BASE_sleeve_01, loc=<0 0 0>, rot=<0 0 0 1>}
 BASE_sleeve_02 {parent=lowerarm_r, loc=<2.4043 4.6446 0.2542>, rot=<-0.1436 -0.9882 -0.0076 -0.0521>}
 dyn_sleeve_02 {parent=BASE_sleeve_02, loc=<0 0 0>, rot=<0 0 0 1>}
 BASE_sleeve_03 {parent=lowerarm_r, loc=<1.3499 -0.7987 9.8522>, rot=<0.106 -0.0307 0.9939 0.0046>}
 dyn_sleeve_03 {parent=BASE_sleeve_03, loc=<0 0 0>, rot=<0 0 0 1>}
 BASE_sleeve_04 {parent=lowerarm_r, loc=<0.6471 -6.9472 -0.7866>, rot=<0 -0.9997 0 -0.0237>}
 dyn_sleeve_04 {parent=BASE_sleeve_04, loc=<0 0 0>, rot=<0 0 0 1>}
 lowerarm_r_zero_scale {parent=lowerarm_r, loc=<0 0 0>, rot=<0 0 0 1>}
 upperarm_twist_01_r {parent=upperarm_r, loc=<-0.0041 0 0>, rot=<0 0 0 -1>}
 BASE_deform_upperarm_bk_base_r {parent=upperarm_twist_01_r, loc=<-1.0814 7.2192 1.9456>, rot=<0 0 0 -1>}
 deform_upperarm_bk_base_r {parent=BASE_deform_upperarm_bk_base_r, loc=<0 0 0>, rot=<0 0 0 1>}
 deform_upperarm_bk_r {parent=deform_upperarm_bk_base_r, loc=<0 0 0>, rot=<0 0 0 1>}
 BASE_deform_upperarm_fr_base_r {parent=upperarm_twist_01_r, loc=<0.3689 -7.0506 2.7652>, rot=<0 0 0 -1>}
 deform_upperarm_fr_base_r {parent=BASE_deform_upperarm_fr_base_r, loc=<0 0 0>, rot=<0 0 0 1>}
 deform_upperarm_fr_r {parent=deform_upperarm_fr_base_r, loc=<0 0 0>, rot=<0 0 0 1>}
 upperarm_twist_01_r_zero_scale {parent=upperarm_twist_01_r, loc=<0 0 0>, rot=<0 0 0 1>}
 upperarm_twist_02_r {parent=upperarm_r, loc=<-13.7941 0 0>, rot=<0 0 0 -1>}
 upperarm_twist_02_r_zero_scale {parent=upperarm_twist_02_r, loc=<0 0 0>, rot=<0 0 0 1>}
 upperarm_r_zero_scale {parent=upperarm_r, loc=<0 0 0>, rot=<0 0 0 1>}
 extra_spine_05 {parent=spine_05, loc=<0 0 0>, rot=<0 0 0 1>}
 body_distance {parent=spine_04, loc=<4.3077 -7.2517 -14.581>, rot=<0.0294 -0.7065 -0.0294 -0.7065>}
 rag_distance {parent=spine_04, loc=<5.1193 -9.7207 -17.0268>, rot=<0.0294 -0.7065 -0.0294 -0.7065>}
 BASE_sweatshirt_R_back {parent=spine_02, loc=<4.0389 -10.8937 10.6844>, rot=<0.0095 -0.4956 0.868 -0.0285>}
 dyn_sweatshirt_R_back_01 {parent=BASE_sweatshirt_R_back, loc=<0 0 0>, rot=<0 0 0 1>}
 dyn_sweatshirt_R_back_02 {parent=dyn_sweatshirt_R_back_01, loc=<7.9225 0 0>, rot=<0 0 0 1>}
 BASE_sweatshirt_M_back {parent=spine_02, loc=<4.2717 -12.8011 0.58>, rot=<0.0164 -0.699 0.7145 -0.0252>}
 dyn_sweatshirt_M_back_01 {parent=BASE_sweatshirt_M_back, loc=<0 0 0>, rot=<0 0 0 1>}
 dyn_sweatshirt_M_back_02 {parent=dyn_sweatshirt_M_back_01, loc=<7.9225 0 0>, rot=<0 0 0 1>}
 BASE_sweatshirt_L_back {parent=spine_02, loc=<4.313 -11.2046 -10.8237>, rot=<0.0236 -0.8922 0.4507 -0.0186>}
 dyn_sweatshirt_L_back_01 {parent=BASE_sweatshirt_L_back, loc=<0 0 0>, rot=<0 0 0 1>}
 dyn_sweatshirt_L_back_02 {parent=dyn_sweatshirt_L_back_01, loc=<7.9225 0 0>, rot=<0 0 0 1>}
 BASE_sweatshirt_R_side_02 {parent=spine_02, loc=<2.2118 4.1012 16.4061>, rot=<0.0298 0.297 0.9535 -0.0418>}
 dyn_sweatshirt_R_side_02_F {parent=BASE_sweatshirt_R_side_02, loc=<0 0 0>, rot=<0 0 0 1>}
 dyn_sweatshirt_R_side_03_F {parent=dyn_sweatshirt_R_side_02_F, loc=<7.9225 0 0>, rot=<0 0 0 1>}
 BASE_sweatshirt_L_side_03 {parent=spine_02, loc=<2.7345 -4.4379 -15.3116>, rot=<0.0526 -0.9603 0.2644 -0.0712>}
 dyn_sweatshirt_L_side_02_B {parent=BASE_sweatshirt_L_side_03, loc=<0 0 0>, rot=<0 0 0 1>}
 dyn_sweatshirt_L_side_03_B {parent=dyn_sweatshirt_L_side_02_B, loc=<8.3782 0 0>, rot=<0 0 0 1>}
 BASE_sweatshirt_R_side_03 {parent=spine_02, loc=<2.8903 -5.1326 17.6553>, rot=<0.0472 -0.2412 0.9691 -0.0203>}
 dyn_sweatshirt_R_side_02_B {parent=BASE_sweatshirt_R_side_03, loc=<0 0 0>, rot=<0 0 0 1>}
 dyn_sweatshirt_R_side_03_B {parent=dyn_sweatshirt_R_side_02_B, loc=<7.9225 0 0>, rot=<0 0 0 1>}
 revers_rotation_spine_01 {parent=spine_01, loc=<0 0 0>, rot=<0 0 0 1>}
 BASE_sweatshirt_L_front_03 {parent=revers_rotation_spine_01, loc=<-1.9524 11.7999 -11.1779>, rot=<-0.0147 -0.7102 -0.7029 -0.0373>}
 dyn_sweatshirt_L_front_03 {parent=BASE_sweatshirt_L_front_03, loc=<0 0 0>, rot=<0 0 0 1>}
 BASE_sweatshirt_M_front_03 {parent=revers_rotation_spine_01, loc=<-2.3635 12.3272 -0.0522>, rot=<-0.0297 -0.7236 -0.6893 -0.022>}
 dyn_sweatshirt_M_front_03 {parent=BASE_sweatshirt_M_front_03, loc=<0 0 0>, rot=<0 0 0 1>}
 BASE_sweatshirt_R_front_03 {parent=revers_rotation_spine_01, loc=<-2.7647 10.276 10.8026>, rot=<0.014 0.701 0.713 0.0045>}
 dyn_sweatshirt_R_front_03 {parent=BASE_sweatshirt_R_front_03, loc=<0 0 0>, rot=<0 0 0 1>}
 BASE_sweatshirt_L_side_03_F {parent=revers_rotation_spine_01, loc=<2.2364 4.8597 -16.3884>, rot=<-0.0163 -0.9639 -0.258 -0.0637>}
 dyn_sweatshirt_L_side_03_F {parent=BASE_sweatshirt_L_side_03_F, loc=<0 0 0>, rot=<0 0 0 1>}
 thigh_l {parent=pelvis, loc=<-2.5693 -0.0109 -7.7501>, rot=<0.0012 0.0572 0.0155 -0.9982>}
 calf_l {parent=thigh_l, loc=<-39.0484 0 0>, rot=<0.0006 -0.0161 -0.0391 -0.9991>}
 foot_l {parent=calf_l, loc=<-42.561 0.1077 0.4292>, rot=<0.0418 -0.0455 0.0343 -0.9975>}
 ball_l {parent=foot_l, loc=<-6.1599 13.8264 0.6806>, rot=<0.0001 -0.0001 -0.7063 -0.7079>}
 calf_twist_02_l {parent=calf_l, loc=<-18.8064 -0.0967 0.0397>, rot=<0.0237 0.0095 -0.0035 -0.9997>}
 calf_twist_01_l {parent=calf_l, loc=<-35.813 -0.0967 0.0397>, rot=<0.0237 0.0095 -0.0035 -0.9997>}
 deform_knee_in_base_l {parent=calf_l, loc=<1.298 -5.2957 -0.0496>, rot=<0 0 0 -1>}
 deform_knee_in_l {parent=deform_knee_in_base_l, loc=<0 0 0>, rot=<0 0 0 1>}
 deform_knee_out_base_l {parent=calf_l, loc=<0 4 0>, rot=<0 0 0 -1>}
 deform_knee_out_l {parent=deform_knee_out_base_l, loc=<0 0 0>, rot=<0 0 0 1>}
 thigh_twist_01_l {parent=thigh_l, loc=<-21.6303 0 0>, rot=<0.0013 0 0 -1>}
 thigh_r {parent=pelvis, loc=<-2.5693 -0.0109 7.7501>, rot=<0.0572 -0.0012 0.9982 0.0155>}
 calf_r {parent=thigh_r, loc=<39.0484 0 0>, rot=<0.0006 -0.0161 -0.0391 -0.9991>}
 foot_r {parent=calf_r, loc=<42.561 -0.1077 -0.4292>, rot=<0.0418 -0.0455 0.0343 -0.9975>}
 ball_r {parent=foot_r, loc=<6.1599 -13.8264 -0.6806>, rot=<0.0001 -0.0001 -0.7063 -0.7079>}
 calf_twist_02_r {parent=calf_r, loc=<18.8064 0.0967 -0.0397>, rot=<0.0237 0.0095 -0.0035 -0.9997>}
 calf_twist_01_r {parent=calf_r, loc=<35.813 0.0967 -0.0397>, rot=<0.0237 0.0095 -0.0035 -0.9997>}
 deform_knee_in_base_r {parent=calf_r, loc=<-1.298 5.2957 0.0496>, rot=<0 0 0 1>}
 deform_knee_in_r {parent=deform_knee_in_base_r, loc=<0 0 0>, rot=<0 0 0 1>}
 deform_knee_out_base_r {parent=calf_r, loc=<0 -4 0>, rot=<0 0 0 -1>}
 deform_knee_out_r {parent=deform_knee_out_base_r, loc=<0 0 0>, rot=<0 0 0 1>}
 thigh_twist_01_r {parent=thigh_r, loc=<21.6303 0 0>, rot=<0.0013 0 0 -1>}
]
variables [
 crouching {type=double, default=0}
 skydive_dive_mode {type=double, default=0}
 temp {type=double, default=0}
 togle_01 {type=double, default=0}
 togle_02 {type=double, default=0}
]
