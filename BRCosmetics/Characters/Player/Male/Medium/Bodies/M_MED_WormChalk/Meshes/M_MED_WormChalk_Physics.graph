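# Physics asset for the WormChalk medium male cosmetic body, with capsules on pelvis, spine, legs and feet plus sweatshirt and top-rag bones.
# Tiny BASE anchor bodies join the spine with free rotation: sweatshirt at spine_01/02, rag at spine_03.
# Angle-limited one- and two-bone dyn chains hang from the sweatshirt anchors.
# The three rag bodies are rotation-locked to their anchors and cross-linked through rag_top_2 with free rotation.
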PhysicsAsset /BRCosmetics/Characters/Player/Male/Medium/Bodies/M_MED_WormChalk/Meshes/M_MED_WormChalk_Physics
bodies {
 BASE_rag_top_1 [{shapeType=Capsule, length=0.505, radius=0.505, shapeName=BASE_rag_top_1_capsule, rotation=<rot 0 -90 0>}]
 BASE_rag_top_2 [{shapeType=Capsule, length=0.505, radius=0.505, shapeName=BASE_rag_top_2_capsule, rotation=<rot 0 -90 0>}]
 BASE_rag_top_3 [{shapeType=Capsule, length=0.505, radius=0.505, shapeName=BASE_rag_top_3_capsule, rotation=<rot 0 -90 0>}]
 BASE_sweatshirt_L_back [{shapeType=Capsule, length=0.505, radius=0.505, shapeName=BASE_sweatshirt_L_back_capsule, rotation=<rot 0 -90 0>}]
 BASE_sweatshirt_L_front_03 [{shapeType=Capsule, length=0.505, radius=0.505, shapeName=BASE_sweatshirt_L_front_03_capsule, rotation=<rot 0 -90 0>}]
 BASE_sweatshirt_L_side_03 [{shapeType=Capsule, length=0.505, radius=0.505, shapeName=BASE_sweatshirt_L_side_03_capsule, rotation=<rot 0 -90 0>}]
 BASE_sweatshirt_L_side_03_F [{shapeType=Capsule, length=0.505, radius=0.505, shapeName=BASE_sweatshirt_L_side_03_F_capsule, rotation=<rot 0 -90 0>}]
 BASE_sweatshirt_M_back [{shapeType=Capsule, length=0.505, radius=0.505, shapeName=BASE_sweatshirt_M_back_capsule, rotation=<rot 0 -90 0>}]
 BASE_sweatshirt_M_front_03 [{shapeType=Capsule, length=0.505, radius=0.505, shapeName=BASE_sweatshirt_M_front_03_capsule, rotation=<rot 0 -90 0>}]
 BASE_sweatshirt_R_back [{shapeType=Capsule, length=0.505, radius=0.505, shapeName=BASE_sweatshirt_R_back_capsule, rotation=<rot 0 -90 0>}]
 BASE_sweatshirt_R_front_03 [{shapeType=Capsule, length=0.505, radius=0.505, shapeName=BASE_sweatshirt_R_front_03_capsule, rotation=<rot 0 -90 0>}]
 BASE_sweatshirt_R_side_02 [{shapeType=Capsule, length=0.505, radius=0.505, shapeName=BASE_sweatshirt_R_side_02_capsule, rotation=<rot 0 -90 0>}]
 BASE_sweatshirt_R_side_03 [{shapeType=Capsule, length=0.505, radius=0.505, shapeName=BASE_sweatshirt_R_side_03_capsule, rotation=<rot 0 -90 0>}]
 calf_l [{shapeType=Capsule, length=21.5603, radius=7.0767, center=<-19.3714 -1.1346 -0.1555>, rotation=<rot 0 86.9498 -89.6241>}]
 calf_r [{shapeType=Capsule, length=21.3551, radius=6.9439, center=<19.5003 1.3877 0.0607>, rotation=<rot 0 87.4813 -89.8454>}]
 dyn_sweatshirt_L_back_01 [{
   shapeType=Capsule
   length=8.2319
   radius=6.2596
   shapeName=dyn_sweatshirt_L_back_01_capsule
   center=<5.437 -0.4838 4.4481>
   rotation=<rot 3.7106 0.6275 -98.5283>
  }]
 dyn_sweatshirt_L_back_02 [{
   shapeType=Capsule
   length=9.6094
   radius=7.9648
   shapeName=dyn_sweatshirt_L_back_02_capsule
   center=<10.3989 -1.2683 7.8103>
   rotation=<rot 80.773 -89.9692 0.0279>
  }]
 dyn_sweatshirt_L_front_03 [{
   shapeType=Capsule
   length=20.219
   radius=6.5293
   shapeName=dyn_sweatshirt_L_front_03_capsule
   center=<7.0181 -1.6895 4.6303>
   rotation=<rot 0.0064 0.0001 72.5736>
  }]
 dyn_sweatshirt_L_side_02_B [{
   shapeType=Capsule
   length=11.384
   radius=6.1175
   shapeName=dyn_sweatshirt_L_side_02_F_capsule
   center=<5.2447 -0.2572 3.9801>
   rotation=<rot -5.057 0.0532 -91.7175>
  }]
 dyn_sweatshirt_L_side_03_B [{
   shapeType=Capsule
   length=10.0179
   radius=6.603
   shapeName=dyn_sweatshirt_L_side_03_F_capsule
   center=<8.6194 -0.1609 3.03>
   rotation=<rot 0.0069 -2.161 -90.0008>
  }]
 dyn_sweatshirt_L_side_03_F [{
   shapeType=Capsule
   length=10.0179
   radius=6.603
   shapeName=dyn_sweatshirt_L_side_03_F_capsule
   center=<8.6194 -0.1609 3.03>
   rotation=<rot 0.0069 -2.161 -90.0008>
  }]
 dyn_sweatshirt_M_back_01 [{
   shapeType=Capsule
   length=11.4509
   radius=6.6938
   shapeName=dyn_sweatshirt_M_back_01_capsule
   center=<5.7389 -0.089 3.6592>
   rotation=<rot 5.3721 0.4214 -90.2078>
  }]
 dyn_sweatshirt_M_back_02 [{
   shapeType=Capsule
   length=10.7333
   radius=7.9105
   shapeName=dyn_sweatshirt_M_back_02_capsule
   center=<10.3671 0.4252 6.7543>
   rotation=<rot -87.8563 -90.0037 0.0037>
  }]
 dyn_sweatshirt_M_front_03 [{
   shapeType=Capsule
   length=15.1436
   radius=6.7759
   shapeName=dyn_sweatshirt_M_front_03_capsule
   center=<11.303 0.0593 4.3417>
   rotation=<rot 0 0 -89.2171>
  }]
 dyn_sweatshirt_R_back_01 [{
   shapeType=Capsule
   length=8.1264
   radius=6.5544
   shapeName=dyn_sweatshirt_R_back_01_capsule
   center=<5.7789 -0.4068 3.9348>
   rotation=<rot -0.1181 -3.501 -89.1831>
  }]
 dyn_sweatshirt_R_back_02 [{
   shapeType=Capsule
   length=9.9304
   radius=8.3145
   shapeName=dyn_sweatshirt_R_back_02_capsule
   center=<10.4279 -0.6777 8.4409>
   rotation=<rot -1.5837 179.6752 -89.5515>
  }]
 dyn_sweatshirt_R_front_03 [{
   shapeType=Capsule
   length=9.7328
   radius=7.0087
   shapeName=dyn_sweatshirt_R_front_03_capsule
   center=<9.3617 -0.1907 4.7873>
   rotation=<rot 4.4618 1.7769 -88.8173>
  }]
 dyn_sweatshirt_R_side_02_B [{
   shapeType=Capsule
   length=11.384
   radius=6.1175
   shapeName=dyn_sweatshirt_L_side_02_F_capsule
   center=<5.2447 -0.2572 3.9801>
   rotation=<rot -5.018 1.323 -91.829>
  }]
 dyn_sweatshirt_R_side_02_F [{
   shapeType=Capsule
   length=11.384
   radius=6.1175
   shapeName=dyn_sweatshirt_L_side_02_F_capsule
   center=<5.2447 -0.2572 3.9801>
   rotation=<rot -5.1317 -2.6372 -91.4786>
  }]
 dyn_sweatshirt_R_side_03_B [{
   shapeType=Capsule
   length=10.0179
   radius=6.603
   shapeName=dyn_sweatshirt_L_side_03_F_capsule
   center=<8.6194 -0.1609 3.03>
   rotation=<rot 0.007 0.743 -90.0004>
  }]
 dyn_sweatshirt_R_side_03_F [{
   shapeType=Capsule
   length=10.0179
   radius=6.603
   shapeName=dyn_sweatshirt_L_side_03_F_capsule
   center=<8.6194 -0.1609 3.03>
   rotation=<rot 0.0069 -2.161 -90.0008>
  }]
 foot_l [{shapeType=Capsule, length=14.9112, radius=9.0682, center=<-4.2241 7.7029 -0.4944>, rotation=<rot 0 17.361 -89.0774>}]
 foot_r [{shapeType=Capsule, length=14.8787, radius=9.0777, center=<4.0549 -7.7863 0.5134>, rotation=<rot 0 16.4599 -89.4203>}]
 pelvis [
  {shapeType=Capsule, length=8.7723, radius=11.0317, shapeName=pelvis_capsule, center=<0 -1.3809 0>}
  {shapeType=Capsule, length=15.5353, radius=10.0223, shapeName=pelvis_capsule, center=<-5.9628 -2.425 0>}
 ]
 rag_top_1 [{
   shapeType=Capsule
   length=5.7381
   radius=3.2692
   shapeName=rag_top_1_capsule
   center=<0.924 -0.5396 1.4281>
   rotation=<rot -8.2531 -90 14.6739>
  }]
 rag_top_2 [{
   shapeType=Capsule
   length=4.7652
   radius=3.4868
   shapeName=rag_top_2_capsule
   center=<-0.5338 -0.516 -0.5147>
   rotation=<rot 0 -171.877 -81.29>
  }]
 rag_top_3 [{
   shapeType=Capsule
   length=4.2059
   radius=3.6909
   shapeName=rag_top_3_capsule
   center=<2.0968 0.8021 2.1185>
   rotation=<rot 5.1957 -90 -38.8187>
  }]
 spine_01 [{shapeType=Capsule, length=14.8222, radius=9.3196, shapeName=spine_01_capsule}]
 spine_02 [{shapeType=Capsule, length=8.8207, radius=9.3522, shapeName=spine_02_capsule}]
 spine_03 [{shapeType=Capsule, length=19.0566, radius=10.2565, shapeName=spine_03_capsule, center=<0 -1.9229 0.3587>, rotation=<rot 10.5677 -90 0.0579>}]
 thigh_l [{shapeType=Capsule, length=21.8866, radius=8.5489, center=<-19.8323 0.2887 1.3329>, rotation=<rot 0 -86.9074 -84.0282>}]
 thigh_r [{shapeType=Capsule, length=21.4589, radius=8.9277, center=<19.7519 0.2035 -1.0873>, rotation=<rot 0 -83.8059 -87.1521>}]
}
constraints [
 {bone1Name=thigh_l, bone2Name=pelvis, swing1LimitDegrees=45, swing2LimitDegrees=45, twistLimitDegrees=45}
 {bone1Name=calf_l, bone2Name=thigh_l, swing1LimitDegrees=45, swing2LimitDegrees=45, twistLimitDegrees=45}
 {bone1Name=foot_l, bone2Name=calf_l, swing1LimitDegrees=45, swing2LimitDegrees=45, twistLimitDegrees=45}
 {bone1Name=thigh_r, bone2Name=pelvis, swing1LimitDegrees=45, swing2LimitDegrees=45, twistLimitDegrees=45}
 {bone1Name=calf_r, bone2Name=thigh_r, swing1LimitDegrees=45, swing2LimitDegrees=45, twistLimitDegrees=45}
 {bone1Name=foot_r, bone2Name=calf_r, swing1LimitDegrees=45, swing2LimitDegrees=45, twistLimitDegrees=45}
 {
  bone1Name=spine_01
  bone2Name=pelvis
  swing1LimitDegrees=45
  swing1Motion=Free
  swing2LimitDegrees=45
  swing2Motion=Free
  twistLimitDegrees=45
  twistMotion=Free
 }
 {
  bone1Name=spine_02
  bone2Name=spine_01
  swing1LimitDegrees=45
  swing1Motion=Free
  swing2LimitDegrees=45
  swing2Motion=Free
  twistLimitDegrees=45
  twistMotion=Free
 }
 {bone1Name=dyn_sweatshirt_L_back_02, bone2Name=dyn_sweatshirt_L_back_01, swing1LimitDegrees=10, swing2LimitDegrees=30, twistLimitDegrees=5}
 {
  bone1Name=BASE_sweatshirt_L_back
  bone2Name=spine_02
  swing1LimitDegrees=45
  swing1Motion=Free
  swing2LimitDegrees=45
  swing2Motion=Free
  twistLimitDegrees=45
  twistMotion=Free
 }
 {bone1Name=dyn_sweatshirt_L_back_01, bone2Name=BASE_sweatshirt_L_back, swing1LimitDegrees=10, swing2LimitDegrees=30, twistLimitDegrees=5}
 {
  bone1Name=BASE_sweatshirt_M_back
  bone2Name=spine_02
  swing1LimitDegrees=45
  swing1Motion=Free
  swing2LimitDegrees=45
  swing2Motion=Free
  twistLimitDegrees=45
  twistMotion=Free
 }
 {bone1Name=dyn_sweatshirt_M_back_01, bone2Name=BASE_sweatshirt_M_back, swing1LimitDegrees=10, swing2LimitDegrees=15, twistLimitDegrees=10}
 {bone1Name=dyn_sweatshirt_M_back_02, bone2Name=dyn_sweatshirt_M_back_01, swing1LimitDegrees=10, swing2LimitDegrees=15, twistLimitDegrees=10}
 {
  bone1Name=BASE_sweatshirt_R_back
  bone2Name=spine_02
  swing1LimitDegrees=45
  swing1Motion=Free
  swing2LimitDegrees=45
  swing2Motion=Free
  twistLimitDegrees=45
  twistMotion=Free
 }
 {bone1Name=dyn_sweatshirt_R_back_01, bone2Name=BASE_sweatshirt_R_back, swing1LimitDegrees=10, swing2LimitDegrees=30, twistLimitDegrees=10}
 {bone1Name=dyn_sweatshirt_R_back_02, bone2Name=dyn_sweatshirt_R_back_01, swing1LimitDegrees=10, swing2LimitDegrees=30, twistLimitDegrees=10}
 {
  bone1Name=spine_03
  bone2Name=spine_02
  swing1LimitDegrees=45
  swing1Motion=Free
  swing2LimitDegrees=45
  swing2Motion=Free
  twistLimitDegrees=45
  twistMotion=Free
 }
 {
  bone1Name=BASE_sweatshirt_L_side_03
  bone2Name=spine_02
  swing1LimitDegrees=45
  swing1Motion=Free
  swing2LimitDegrees=45
  swing2Motion=Free
  twistLimitDegrees=45
  twistMotion=Free
 }
 {
  bone1Name=dyn_sweatshirt_L_side_02_B
  bone2Name=BASE_sweatshirt_L_side_03
  swing1LimitDegrees=10
  swing2LimitDegrees=25
  twistLimitDegrees=10
 }
 {
  bone1Name=dyn_sweatshirt_L_side_03_B
  bone2Name=dyn_sweatshirt_L_side_02_B
  swing1LimitDegrees=10
  swing2LimitDegrees=25
  twistLimitDegrees=10
 }
 {
  bone1Name=dyn_sweatshirt_L_side_03_F
  bone2Name=BASE_sweatshirt_L_side_03_F
  swing1LimitDegrees=10
  swing2LimitDegrees=25
  twistLimitDegrees=40
 }
 {
  bone1Name=BASE_sweatshirt_R_side_02
  bone2Name=spine_02
  swing1LimitDegrees=45
  swing1Motion=Free
  swing2LimitDegrees=45
  swing2Motion=Free
  twistLimitDegrees=45
  twistMotion=Free
 }
 {
  bone1Name=dyn_sweatshirt_R_side_02_F
  bone2Name=BASE_sweatshirt_R_side_02
  swing1LimitDegrees=10
  swing2LimitDegrees=25
  twistLimitDegrees=10
 }
 {
  bone1Name=dyn_sweatshirt_R_side_03_F
  bone2Name=dyn_sweatshirt_R_side_02_F
  swing1LimitDegrees=10
  swing2LimitDegrees=25
  twistLimitDegrees=10
 }
 {
  bone1Name=BASE_sweatshirt_R_side_03
  bone2Name=spine_02
  swing1LimitDegrees=45
  swing1Motion=Free
  swing2LimitDegrees=45
  swing2Motion=Free
  twistLimitDegrees=45
  twistMotion=Free
 }
 {
  bone1Name=dyn_sweatshirt_R_side_02_B
  bone2Name=BASE_sweatshirt_R_side_03
  swing1LimitDegrees=10
  swing2LimitDegrees=25
  twistLimitDegrees=10
 }
 {
  bone1Name=dyn_sweatshirt_R_side_03_B
  bone2Name=dyn_sweatshirt_R_side_02_B
  swing1LimitDegrees=10
  swing2LimitDegrees=25
  twistLimitDegrees=10
 }
 {bone1Name=dyn_sweatshirt_L_front_03, bone2Name=BASE_sweatshirt_L_front_03, swing1LimitDegrees=5, swing2LimitDegrees=45, twistLimitDegrees=5}
 {
  bone1Name=BASE_sweatshirt_L_front_03
  bone2Name=spine_01
  swing1LimitDegrees=45
  swing1Motion=Free
  swing2LimitDegrees=45
  swing2Motion=Free
  twistLimitDegrees=45
  twistMotion=Free
 }
 {bone1Name=dyn_sweatshirt_M_front_03, bone2Name=BASE_sweatshirt_M_front_03, swing1LimitDegrees=5, swing2LimitDegrees=45, twistLimitDegrees=5}
 {
  bone1Name=BASE_sweatshirt_M_front_03
  bone2Name=spine_01
  swing1LimitDegrees=45
  swing1Motion=Free
  swing2LimitDegrees=45
  swing2Motion=Free
  twistLimitDegrees=45
  twistMotion=Free
 }
 {bone1Name=dyn_sweatshirt_R_front_03, bone2Name=BASE_sweatshirt_R_front_03, swing1LimitDegrees=5, swing2LimitDegrees=45, twistLimitDegrees=5}
 {
  bone1Name=BASE_sweatshirt_R_front_03
  bone2Name=spine_01
  swing1LimitDegrees=45
  swing1Motion=Free
  swing2LimitDegrees=45
  swing2Motion=Free
  twistLimitDegrees=45
  twistMotion=Free
 }
 {
  bone1Name=BASE_sweatshirt_L_side_03_F
  bone2Name=spine_01
  swing1LimitDegrees=45
  swing1Motion=Free
  swing2LimitDegrees=45
  swing2Motion=Free
  twistLimitDegrees=45
  twistMotion=Free
 }
 {
  bone1Name=BASE_rag_top_1
  bone2Name=spine_03
  swing1LimitDegrees=45
  swing1Motion=Free
  swing2LimitDegrees=45
  swing2Motion=Free
  twistLimitDegrees=45
  twistMotion=Free
 }
 {
  bone1Name=rag_top_1
  bone2Name=BASE_rag_top_1
  swing1LimitDegrees=45
  swing1Motion=Locked
  swing2LimitDegrees=45
  swing2Motion=Locked
  twistLimitDegrees=45
  twistMotion=Locked
 }
 {
  bone1Name=BASE_rag_top_2
  bone2Name=spine_03
  swing1LimitDegrees=45
  swing1Motion=Free
  swing2LimitDegrees=45
  swing2Motion=Free
  twistLimitDegrees=45
  twistMotion=Free
 }
 {
  bone1Name=rag_top_2
  bone2Name=BASE_rag_top_2
  swing1LimitDegrees=45
  swing1Motion=Locked
  swing2LimitDegrees=45
  swing2Motion=Locked
  twistLimitDegrees=45
  twistMotion=Locked
 }
 {
  bone1Name=BASE_rag_top_3
  bone2Name=spine_03
  swing1LimitDegrees=45
  swing1Motion=Free
  swing2LimitDegrees=45
  swing2Motion=Free
  twistLimitDegrees=45
  twistMotion=Free
 }
 {
  bone1Name=rag_top_3
  bone2Name=BASE_rag_top_3
  swing1LimitDegrees=45
  swing1Motion=Locked
  swing2LimitDegrees=45
  swing2Motion=Locked
  twistLimitDegrees=45
  twistMotion=Locked
 }
 {
  bone1Name=rag_top_3
  bone2Name=rag_top_2
  swing1LimitDegrees=45
  swing1Motion=Free
  swing2LimitDegrees=45
  swing2Motion=Free
  twistLimitDegrees=45
  twistMotion=Free
 }
 {
  bone1Name=rag_top_1
  bone2Name=rag_top_2
  swing1LimitDegrees=45
  swing1Motion=Free
  swing2LimitDegrees=45
  swing2Motion=Free
  twistLimitDegrees=45
  twistMotion=Free
 }
]
