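# Post-process style AnimBlueprint for the WormChalk male body cosmetic, adding secondary motion to the pose copied from its attached parent mesh.
# A pre-sim Control Rig feeds RigidBody simulation, four AnimDynamics sleeve bones with planar forearm limits, then a post-sim Control Rig.
# Crouch and skydive states drive rig inputs and sleeve alpha.
# The update also computes gravity override and, outside the frontend, feeds WPO and wind parameters to body and overlay materials.
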AnimBlueprint /BRCosmetics/Characters/Player/Male/Medium/Bodies/M_MED_WormChalk/Meshes/M_MED_WormChalk_AnimBP
target_skeleton=&/M_MED_WormChalk_Skeleton
graphs [
 AnimGraph {nodes [
   n1 = Control Rig!(
    Source=Copy Pose From Mesh() {bUseAttachedParent=true}
    Alpha
    crouching:double=Get crouching_float() -> (crouching_float:double)
    skydive_dive_mode:double=Get skydive_dive_mode() -> (skydive_dive_mode:double)
   ) -> (Pose) {
    alphaBoolBlend {blendOption=Linear}
    controlRigAssetReference {blueprintRigClass=&/M_MED_WormChalk_CtrlRig_presim.M_MED_WormChalk_CtrlRig_presim_C}
   }
   n2 = RigidBody!(ComponentPose=Local To Component(LocalPose=@n1), bAlphaBoolEnabled:bool, OverrideWorldGravity:Vector="0,0,0", ExternalForce) -> (Pose) {
    alphaInputType=Bool
    overridePhysicsAsset=&/M_MED_WormChalk_Physics
    simulationSpace=BaseBoneSpace
    alphaBoolBlend {blendInTime=0.2, blendOption=Linear, blendOutTime=0.2}
    baseBoneRef {boneName=spine_01}
    componentAppliedLinearAccClamp=<1200 1200 1200>
    componentLinearAccScale=<1 1 1>
    componentLinearVelScale=<1 1 1>
    simSpaceSettings {
     maxAngularAcceleration=100
     maxAngularVelocity=2
     maxLinearAcceleration=100
     maxLinearVelocity=160
     velocityScaleZ=0
     worldAlpha=0.75
     externalLinearDragV=<1 1 1>
    }
   }
   n3 = Get skydive() -> (skydive:double)
   n4 = Anim Dynamics - Bone: dyn_sleeve_01!(ComponentPose=@n2, Alpha=@n3, GravityOverride:Vector="0,0,0", ExternalForce="0,0,0") -> (Pose) {
    angularDampingOverride=1
    angularSpringConstant=5000
    bAngularSpring=true
    bOverrideAngularDamping=true
    bOverrideLinearDamping=true
    bUseGravityOverride=true
    linearDampingOverride=1
    alphaBoolBlend {blendOption=Linear}
    boundBone {boneName=dyn_sleeve_01}
    componentAppliedLinearAccClamp=<200 200 200>
    componentLinearAccScale=<0.1 0.1 0.1>
    componentLinearVelScale=<0.01 0.01 0.01>
    physicsBodyDefinitions [{
      boundBone {boneName=dyn_sleeve_01}
      boxExtents=<10 10 10>
      constraintSetup {coneAngle=20, angularLimitsMax=<5 0 5>, angularLimitsMin=<-5 -10 -5>, angularTarget=<1 0 0>}
      localJointOffset=<1 0 0>
     }]
    planarLimits [{drivingBone {boneName=lowerarm_r}, planeTransform {location=<0 0 0>, rotation=<rot 0 0 180>}}]
   }
   n5 = Anim Dynamics - Bone: dyn_sleeve_02!(ComponentPose=@n4, Alpha=@n3, GravityOverride:Vector="0,0,0", ExternalForce="0,0,0") -> (Pose) {
    angularDampingOverride=0.99
    angularSpringConstant=1000
    bAngularSpring=true
    bOverrideAngularDamping=true
    bOverrideLinearDamping=true
    bUseGravityOverride=true
    linearDampingOverride=0.99
    simulationSpace=RootRelative
    alphaBoolBlend {blendOption=Linear}
    boundBone {boneName=dyn_sleeve_02}
    componentAppliedLinearAccClamp=<200 200 200>
    componentLinearAccScale=<0.1 0.1 0.1>
    componentLinearVelScale=<0.01 0.01 0.01>
    physicsBodyDefinitions [{
      boundBone {boneName=dyn_sleeve_02}
      boxExtents=<10 10 10>
      constraintSetup {coneAngle=20, angularLimitsMax=<5 5 5>, angularLimitsMin=<-5 -5 -20>, angularTarget=<1 0 0>}
      localJointOffset=<1 0 0>
     }]
    planarLimits [{drivingBone {boneName=lowerarm_r}, planeTransform {location=<0 0 0>, rotation=<rot 0 0 90>}}]
    relativeSpaceBone {boneName=upperarm_r}
   }
   n6 = Anim Dynamics - Bone: dyn_sleeve_03!(ComponentPose=@n5, Alpha=@n3, GravityOverride:Vector="0,0,0", ExternalForce="0,0,0") -> (Pose) {
    angularDampingOverride=1
    angularSpringConstant=5000
    bAngularSpring=true
    bOverrideAngularDamping=true
    bOverrideLinearDamping=true
    bUseGravityOverride=true
    bUseSphericalLimits=true
    linearDampingOverride=1
    alphaBoolBlend {blendOption=Linear}
    boundBone {boneName=dyn_sleeve_03}
    componentLinearAccScale=<0.5 0.5 0.5>
    componentLinearVelScale=<0.5 0.5 0.5>
    physicsBodyDefinitions [{
      boundBone {boneName=dyn_sleeve_03}
      boxExtents=<10 10 10>
      constraintSetup {coneAngle=20, angularLimitsMax=<5 20 5>, angularLimitsMin=<-5 0 -5>, angularTarget=<1 0 0>}
      localJointOffset=<1 0 0>
     }]
    planarLimits [{drivingBone {boneName=lowerarm_r}, planeTransform {location=<0 0 3>, rotation=<rot 0 0 0>}}]
   }
   n7 = Anim Dynamics - Bone: dyn_sleeve_04!(ComponentPose=@n6, Alpha=@n3, GravityOverride:Vector="0,0,0", ExternalForce="0,0,0") -> (Pose) {
    angularDampingOverride=1
    angularSpringConstant=5000
    bAngularSpring=true
    bOverrideAngularDamping=true
    bOverrideLinearDamping=true
    bUseGravityOverride=true
    bUseSphericalLimits=true
    linearDampingOverride=1
    simulationSpace=BoneRelative
    alphaBoolBlend {blendOption=Linear}
    boundBone {boneName=dyn_sleeve_04}
    componentAppliedLinearAccClamp=<200 200 200>
    componentLinearAccScale=<0.1 0.1 0.1>
    componentLinearVelScale=<0.01 0.01 0.01>
    physicsBodyDefinitions [{
      boundBone {boneName=dyn_sleeve_04}
      boxExtents=<10 10 10>
      constraintSetup {coneAngle=20, angularLimitsMax=<5 5 0>, angularLimitsMin=<-5 -5 -7>, angularTarget=<1 0 0>}
      localJointOffset=<1 0 0>
     }]
    planarLimits [{drivingBone {boneName=lowerarm_r}, planeTransform {location=<0 0 0>, rotation=<rot 0 0 -90>}}]
    relativeSpaceBone {boneName=upperarm_r}
   }
   n8 = Control Rig(Source=Component To Local(ComponentPose=@n7), crouching=Get crouching_float(), skydive_dive_mode=Get skydive_dive_mode()) {
    alphaBoolBlend {blendOption=Linear}
    controlRigAssetReference {blueprintRigClass=&/M_MED_WormChalk_CtrlRig_postsim.M_MED_WormChalk_CtrlRig_postsim_C}
   }
   Output Pose(Result=@n8) {layerGroup=None}
  ]}
 EventGraph {nodes [
   Event BlueprintUpdateAnimation()
   > n1 = Sequence()
   > Set crouching_float(crouching_float:double="To Float (Boolean)"(InBool=Get bIsCrouching())) -> (Output_Get:double)
   > Set skydive(skydive:double=SelectFloat(A=0.5, B=1, bPickA=Get bIsSkydiveFloating())) -> (Output_Get:double)
   > Set skydive_dive_mode(skydive_dive_mode:double="To Float (Boolean)"(InBool=Get bIsSkydiveDiveMode())) -> (Output_Get:double)
   n2 = Gravity Override State Function DEPRECATED(
    execute=@n1.then_1
    CharacterPartAnimInstance=Self-Reference()
    GameTimeInSeconds=GetGameTimeInSeconds()
    GravityOverrideParams=Get Gravity Override Params()
   )
   > Set Gravity Override(Gravity Override=@n2)
   > Is Valid(InputObject=Get PlayerPawn() -> (PlayerPawn:Fort Player Pawn Athena Object Reference))
   > n3 = Branch(Condition=Get bIsInFrontEnd())
   Set WPO Globals Function(
    execute=@n3.else
    Material=Get Mat Instance Dynamic()
    Overlay Material=Get Overlay Mat Instance Dynamic()
    Skeletal Mesh=Get SkelMesh()
    "Local Noise Transform  Location"=Get Local Noise Transform Location() -> (Local Noise Transform Location:Vector)
    Local Noise Offset=Get Local Noise Offset()
   )
   > SetScalarParameterValue(
    self=[Get Mat Instance Dynamic(), Get Overlay Mat Instance Dynamic()]
    ParameterName=ClothSim_WindStrength
    Value="To Float (Boolean)"(InBool=Get bIsSkydiveFloating())
   )
   Event BlueprintInitializeAnimation()
   n4 = TryGetPawnOwner()
   > Is Valid(InputObject=@n4)
   > n5 = Cast To FortPlayerPawnAthena(Object=@n4)
   > Set PlayerPawn(PlayerPawn:Fort Player Pawn Athena Object Reference=@n5) -> (Output_Get:Fort Player Pawn Athena Object Reference)
   n6 = Event OnCharacterCustomizationLoaded()
   > n7 = Set SkelMesh(SkelMesh=Get Array Item with EFort Custom Part Type(Array=@n6.CustomMeshParts))
   > n8 = Get Dynamic Material Instance(SkeletalMeshComponent=@n7)
   > Set Mat Instance Dynamic(Mat Instance Dynamic=@n8)
   > n9 = GetOrCreateDynamicOverlayMaterial(MeshComponent=@n7, MaterialName=None)
   > Set Overlay Mat Instance Dynamic(Overlay Mat Instance Dynamic=@n9)
  ]}
]
variables [
 crouching_float {type=number}
 Gravity Override {type=Vector}
 Gravity Override Params {type=GravityOverrideParamsStruct}
 Local Noise Offset {type=Vector}
 Local Noise Transform Location {type=Vector}
 Mat Instance Dynamic {type=MaterialInstanceDynamic}
 Overlay Mat Instance Dynamic {type=MaterialInstanceDynamic}
 PlayerPawn {type=FortPlayerPawnAthena}
 R1 Velocity Transform Location {type=Vector}
 R1 Velocity Transform Velocity {type=Vector}
 SkelMesh {type=SkeletalMeshComponent}
 skydive {type=number}
 skydive_dive_mode {type=number}
 Velocity Interp Speed {type=number}
 WPO Pho Wind Intensity {type=number}
]
