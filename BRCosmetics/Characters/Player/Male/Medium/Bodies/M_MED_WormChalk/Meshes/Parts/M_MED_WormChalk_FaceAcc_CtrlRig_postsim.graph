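# Post-simulation Control Rig for the face accessory of the WormChalk male medium character, adjusting hair bones when a backpack is worn.
# It converts the backpack_wearing bool to a float that weights six chained parent constraints.
# These pull segments 02 to 04 of the left and middle dyn hair chains toward weighted blends of spine, neck, head and clavicle bones, overriding their incoming pose.
# Right and front hair chains are not constrained.
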ControlRigBlueprint /BRCosmetics/Characters/Player/Male/Medium/Bodies/M_MED_WormChalk/Meshes/Parts/M_MED_WormChalk_FaceAcc_CtrlRig_postsim
graphs [
 RigVMFunctionLibrary {links [], nodes []}
 RigVMModel {nodes [
   RigUnit_BeginExecution = &RigUnit_BeginExecution()
   RigVMFunction_MathBoolToFloat = &RigVMFunction_MathBoolToFloat(Value=RigVMVariableNode() {variable=backpack_wearing}.Value)
   &RigUnit_ParentConstraint(
    AdvancedSettings=(RotationOrderForFilter=XZY)
    Child=(Type=Bone,Name="dyn_L_hair_02")
    Parents="((Item=(Type=Bone,Name=\"head\"),Weight=0.013),(Item=(Type=Bone,Name=\"spine_03\"),Weight=0.0235),(Item=(Type=Bone,Name=\"neck_01\"),Weight=0.074),(Item=(Type=Bone,Name=\"clavicle_l\"),Weight=0.1581),(Item=(Type=Bone,Name=\"neck_02\"),Weight=0.1699),(Item=(Type=Bone ...(truncated)"
    ExecutePin=@RigUnit_BeginExecution.ExecutePin
    Weight=@RigVMFunction_MathBoolToFloat.Result
   )
   > &RigUnit_ParentConstraint(
    AdvancedSettings=(RotationOrderForFilter=XZY)
    Child=(Type=Bone,Name="dyn_L_hair_03")
    Parents="((Item=(Type=Bone,Name=\"head\"),Weight=0.011),(Item=(Type=Bone,Name=\"spine_03\"),Weight=0.0387),(Item=(Type=Bone,Name=\"neck_01\"),Weight=0.0747),(Item=(Type=Bone,Name=\"clavicle_l\"),Weight=0.1193),(Item=(Type=Bone,Name=\"neck_02\"),Weight=0.1648),(Item=(Type=Bon ...(truncated)"
    Weight=@RigVMFunction_MathBoolToFloat.Result
   )
   > &RigUnit_ParentConstraint(
    AdvancedSettings=(RotationOrderForFilter=XZY)
    Child=(Type=Bone,Name="dyn_L_hair_04")
    Parents=((Item=(Type=Bone,Name="neck_02"),Weight=0.0121),(Item=(Type=Bone,Name="spine_04"),Weight=0.087),(Item=(Type=Bone,Name="spine_03"),Weight=0.1153),(Item=(Type=Bone,Name="clavicle_l"),Weight=0.297),(Item=(Type=Bone,Name="spine_05"),Weight=0.4886))
    Weight=@RigVMFunction_MathBoolToFloat.Result
   )
   > &RigUnit_ParentConstraint(
    AdvancedSettings=(RotationOrderForFilter=XZY)
    Child=(Type=Bone,Name="dyn_M_hair_02")
    Parents=((Item=(Type=Bone,Name="head"),Weight=0.0037),(Item=(Type=Bone,Name="spine_03"),Weight=0.0436),(Item=(Type=Bone,Name="neck_01"),Weight=0.0958),(Item=(Type=Bone,Name="neck_02"),Weight=0.1889),(Item=(Type=Bone,Name="spine_05"),Weight=0.6681))
    Weight=@RigVMFunction_MathBoolToFloat.Result
   )
   > &RigUnit_ParentConstraint(
    AdvancedSettings=(RotationOrderForFilter=XZY)
    Child=(Type=Bone,Name="dyn_M_hair_03")
    Parents=((Item=(Type=Bone,Name="head"),Weight=0.0037),(Item=(Type=Bone,Name="spine_03"),Weight=0.0436),(Item=(Type=Bone,Name="neck_01"),Weight=0.0958),(Item=(Type=Bone,Name="neck_02"),Weight=0.1889),(Item=(Type=Bone,Name="spine_05"),Weight=0.6681))
    Weight=@RigVMFunction_MathBoolToFloat.Result
   )
   > &RigUnit_ParentConstraint(
    AdvancedSettings=(RotationOrderForFilter=XZY)
    Child=(Type=Bone,Name="dyn_M_hair_04")
    Parents=((Item=(Type=Bone,Name="neck_01"),Weight=0.0224),(Item=(Type=Bone,Name="neck_02"),Weight=0.0378),(Item=(Type=Bone,Name="spine_04"),Weight=0.0706),(Item=(Type=Bone,Name="spine_03"),Weight=0.0826),(Item=(Type=Bone,Name="spine_05"),Weight=0.7866))
    Weight=@RigVMFunction_MathBoolToFloat.Result
   )
  ]}
]
hierarchy [
 root {loc=<0 0 0>, rot=<0 0 0 1>}
 pelvis {parent=root, loc=<0 -0.7071 93.1191>, rot=<0.0127 0.707 -0.0127 -0.707>}
 spine_01 {parent=pelvis, loc=<6.7494 -0.6388 0>, rot=<0 0 0.034 -0.9994>}
 spine_02 {parent=spine_01, loc=<7.252 -0.035 0>, rot=<0 0 0.0238 -0.9997>}
 spine_03 {parent=spine_02, loc=<6.0106 0 0>, rot=<0 0 0.0149 -0.9999>}
 spine_04 {parent=spine_03, loc=<8.1023 0 0>, rot=<0 0 -0.0132 -0.9999>}
 spine_05 {parent=spine_04, loc=<10.0142 0 0>, rot=<0 0 -0.0015 -1>}
 neck_01 {parent=spine_05, loc=<11.3138 0.1529 0>, rot=<0 0 -0.2003 -0.9797>}
 neck_02 {parent=neck_01, loc=<4.5627 0 0>, rot=<0 0 0.0504 -0.9987>}
 head {parent=neck_02, loc=<5.3914 0 0>, rot=<0 0 0.115 -0.9934>}
 BASE_R_hair {parent=head, loc=<-1.1513 -4.4469 9.3915>, rot=<-0.2139 -0.9711 0.0929 0.0508>}
 dyn_R_hair_01 {parent=BASE_R_hair, loc=<0 0 0>, rot=<0 0 0 1>}
 dyn_R_hair_02 {parent=dyn_R_hair_01, loc=<5.6139 0 0>, rot=<-0.0175 0.0611 -0.2753 -0.9593>}
 dyn_R_hair_03 {parent=dyn_R_hair_02, loc=<5.8755 0.6079 -0.8035>, rot=<0.0115 0.1211 0.0939 -0.9881>}
 dyn_R_hair_04 {parent=dyn_R_hair_03, loc=<10.464 0 0>, rot=<0 0 0 1>}
 BASE_M_hair {parent=head, loc=<9.5468 -7.5093 0>, rot=<0 0 0.9951 -0.0993>}
 dyn_M_hair_01 {parent=BASE_M_hair, loc=<0 0 0>, rot=<0 0 0 1>}
 dyn_M_hair_02 {parent=dyn_M_hair_01, loc=<8.7791 0 0>, rot=<0 0 -0.0991 -0.9951>}
 dyn_M_hair_03 {parent=dyn_M_hair_02, loc=<8.2692 0 0>, rot=<0 0 0.0059 -1>}
 dyn_M_hair_04 {parent=dyn_M_hair_03, loc=<8.7845 0 0>, rot=<0 0 0.1752 -0.9845>}
 BASE_L_hair {parent=head, loc=<9.5646 -5.0822 -5.5493>, rot=<-0.0721 -0.016 0.9964 -0.0413>}
 dyn_L_hair_01 {parent=BASE_L_hair, loc=<0 0 0>, rot=<0 0 0 1>}
 dyn_L_hair_02 {parent=dyn_L_hair_01, loc=<8.7791 0 0>, rot=<0 0 -0.1979 -0.9802>}
 dyn_L_hair_03 {parent=dyn_L_hair_02, loc=<8.2692 0 0>, rot=<0 0 -0.0305 -0.9995>}
 dyn_L_hair_04 {parent=dyn_L_hair_03, loc=<8.7845 0 0>, rot=<0 0 0.201 -0.9796>}
 hair_tip {parent=dyn_L_hair_04, loc=<9.9717 3.6682 -0.0131>, rot=<0 0 -0.0552 -0.9985>}
 extra_dyn_L_hair_02 {parent=dyn_L_hair_02, loc=<0 0 0>, rot=<0 0 0 1>}
 BASE_M_front_hair {parent=head, loc=<11.3367 11.0696 -0.1103>, rot=<0.0041 -0.9994 0.0345 -0.0001>}
 dyn_M_front_hair_01 {parent=BASE_M_front_hair, loc=<0 0 0>, rot=<-0.0122 0 0 -0.9999>}
 dyn_M_front_hair_02 {parent=dyn_M_front_hair_01, loc=<3.3724 0 0>, rot=<0 0 0 1>}
 M_front_hair_01 {parent=BASE_M_front_hair, loc=<0 0 0>, rot=<-0.0122 0 0 -0.9999>}
 M_front_hair_02 {parent=M_front_hair_01, loc=<3.3724 0 0>, rot=<0 0 0 1>}
 BASE_R_front_hair {parent=head, loc=<11.3863 5.0703 7.3451>, rot=<0.0035 -0.8478 -0.5302 0.0022>}
 dyn_R_front_hair_01 {parent=BASE_R_front_hair, loc=<0 0 0>, rot=<-0.0122 0 0 -0.9999>}
 dyn_R_front_hair_02 {parent=dyn_R_front_hair_01, loc=<3.3724 0 0>, rot=<0 0 0 1>}
 dyn_R_front_hair_03 {parent=dyn_R_front_hair_02, loc=<3.5294 0 0>, rot=<0 0 0 1>}
 R_front_hair_01 {parent=BASE_R_front_hair, loc=<0 0 0>, rot=<-0.0122 0 0 -0.9999>}
 R_front_hair_02 {parent=R_front_hair_01, loc=<3.3724 0 0>, rot=<0 0 0 1>}
 R_front_hair_03 {parent=R_front_hair_02, loc=<3.5294 0 0>, rot=<0 0 0 1>}
 BASE_L_front_hair {parent=head, loc=<11.395 4.0081 -6.9894>, rot=<0.0032 -0.7721 0.6355 -0.0026>}
 L_front_hair_01 {parent=BASE_L_front_hair, loc=<0 0 0>, rot=<0 0 0 1>}
 L_front_hair_02 {parent=L_front_hair_01, loc=<3.3724 0 0>, rot=<0 0 0 1>}
 L_front_hair_03 {parent=L_front_hair_02, loc=<3.5294 0 0>, rot=<0 0 0 1>}
 clavicle_l {parent=spine_05, loc=<4.3575 -0.1466 -2.3223>, rot=<0.0465 -0.7289 -0.009 -0.6829>}
 upperarm_l {parent=clavicle_l, loc=<15.5111 0 0>, rot=<0.0226 -0.3234 -0.0088 -0.9459>}
 clavicle_r {parent=spine_05, loc=<4.3575 -0.1466 2.3223>, rot=<0.7289 0.0465 -0.6829 0.009>}
 upperarm_r {parent=clavicle_r, loc=<-15.5111 0 0>, rot=<0.0226 -0.3234 -0.0088 -0.9459>}
 lowerarm_r {parent=upperarm_r, loc=<-26.6514 0 0>, rot=<0.0391 -0.0099 -0.2453 -0.9686>}
 BASE_sleeve_03 {parent=lowerarm_r, loc=<1.3499 -0.7987 9.8522>, rot=<0.106 -0.0307 0.9939 0.0046>}
 dyn_sleeve_03 {parent=BASE_sleeve_03, loc=<0 0 0>, rot=<0 0 0 1>}
 BASE_sleeve_02 {parent=lowerarm_r, loc=<2.4043 4.6446 0.2542>, rot=<-0.1436 -0.9882 -0.0076 -0.0521>}
 dyn_sleeve_02 {parent=BASE_sleeve_02, loc=<0 0 0>, rot=<0 0 0 1>}
 BASE_sleeve_01 {parent=lowerarm_r, loc=<2.083 -0.929 -5.2867>, rot=<-0.0527 -0.993 -0.0293 -0.1011>}
 dyn_sleeve_01 {parent=BASE_sleeve_01, loc=<0 0 0>, rot=<0 0 0 1>}
 BASE_sleeve_04 {parent=lowerarm_r, loc=<0.6471 -6.9472 -0.7866>, rot=<0 -0.9997 0 -0.0237>}
 dyn_sleeve_04 {parent=BASE_sleeve_04, loc=<0 0 0>, rot=<0 0 0 1>}
 BASE_sweatshirt_R_back {parent=spine_02, loc=<4.0389 -10.8937 10.6844>, rot=<0.0095 -0.4956 0.868 -0.0285>}
 dyn_sweatshirt_R_back_01 {parent=BASE_sweatshirt_R_back, loc=<0 0 0>, rot=<0 0 0 1>}
 dyn_sweatshirt_R_back_02 {parent=dyn_sweatshirt_R_back_01, loc=<7.9225 0 0>, rot=<0 0 0 1>}
 BASE_sweatshirt_M_back {parent=spine_02, loc=<4.2717 -12.8011 0.58>, rot=<0.0164 -0.699 0.7145 -0.0252>}
 dyn_sweatshirt_M_back_01 {parent=BASE_sweatshirt_M_back, loc=<0 0 0>, rot=<0 0 0 1>}
 dyn_sweatshirt_M_back_02 {parent=dyn_sweatshirt_M_back_01, loc=<7.9225 0 0>, rot=<0 0 0 1>}
 BASE_sweatshirt_L_back {parent=spine_02, loc=<4.313 -11.2046 -10.8237>, rot=<0.0236 -0.8922 0.4507 -0.0186>}
 dyn_sweatshirt_L_back_01 {parent=BASE_sweatshirt_L_back, loc=<0 0 0>, rot=<0 0 0 1>}
 dyn_sweatshirt_L_back_02 {parent=dyn_sweatshirt_L_back_01, loc=<7.9225 0 0>, rot=<0 0 0 1>}
]
variables [backpack_wearing {type=bool, default=False}]
