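# Physics asset for the hair of the M_MED_WormChalk male player cosmetic's face accessory part, built from capsule bodies.
# Six hair chains (left, middle, right, each front and main) hang from small BASE bodies on the head, with tight per-link swing and twist angles.
# Spine, head, clavicle and upper-arm capsules approximate the upper body.
# Sweatshirt-back and right-sleeve bodies are also included, joined, like the body and BASE bodies, by free angular constraints.
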PhysicsAsset /BRCosmetics/Characters/Player/Male/Medium/Bodies/M_MED_WormChalk/Meshes/Parts/M_MED_WormChalk_FaceAcc_Physics_hair
bodies {
 BASE_L_front_hair [{shapeType=Capsule, length=0.505, radius=0.505, shapeName=BASE_L_front_hair_capsule, rotation=<rot 0 -90 0>}]
 BASE_L_hair [{shapeType=Capsule, length=0.505, radius=0.505, shapeName=BASE_L_hair_capsule, rotation=<rot 0 -90 0>}]
 BASE_M_front_hair [{shapeType=Capsule, length=0.505, radius=0.505, shapeName=BASE_M_front_hair_capsule, rotation=<rot 0 -90 0>}]
 BASE_M_hair [{shapeType=Capsule, length=0.505, radius=0.505, shapeName=BASE_M_hair_capsule, rotation=<rot 0 -90 0>}]
 BASE_R_front_hair [{shapeType=Capsule, length=0.505, radius=0.505, shapeName=BASE_R_front_hair_capsule, rotation=<rot 0 -90 0>}]
 BASE_R_hair [{shapeType=Capsule, length=0.505, radius=0.505, shapeName=BASE_R_hair_capsule, rotation=<rot 0 -90 0>}]
 L_front_hair_01 [{shapeType=Capsule, length=5.0967, radius=2.2564, shapeName=L_front_hair_01_capsule, center=<2.2548 1.4762 -0.0605>}]
 L_front_hair_02 [{
   shapeType=Capsule
   length=1.8803
   radius=4.0012
   shapeName=L_front_hair_02_capsule
   center=<1.8658 3.0398 -0.678>
   rotation=<rot 10.6068 -90 0.5669>
  }]
 L_front_hair_03 [{
   shapeType=Capsule
   length=3.9091
   radius=4.0778
   shapeName=L_front_hair_03_capsule
   center=<4.2515 2.2639 0.1395>
   rotation=<rot 0 -103.9446 -0.0316>
  }]
 clavicle_l [{shapeType=Capsule, length=10.9431, radius=7.9473, shapeName=clavicle_l_capsule, center=<7.9003 0 -1.7012>, rotation=<rot 0 -90 -89.6884>}]
 clavicle_r [{shapeType=Capsule, length=10.9431, radius=7.9473, shapeName=clavicle_l_capsule, center=<-7.9003 0 1.7012>, rotation=<rot 0 -90 90.3116>}]
 dyn_L_hair_01 [{
   shapeType=Capsule
   length=6.1214
   radius=4.7349
   shapeName=dyn_L_hair_01_capsule
   center=<7.1227 2.8693 0.5698>
   rotation=<rot 80.7007 128.2139 111.1719>
  }]
 dyn_L_hair_02 [{
   shapeType=Capsule
   length=4.1456
   radius=5.0493
   shapeName=dyn_L_hair_02_capsule
   center=<6.5376 0.9731 1.6997>
   rotation=<rot 4.5046 96.6859 -91.1612>
  }]
 dyn_L_hair_03 [{
   shapeType=Capsule
   length=7.5115
   radius=3.9376
   shapeName=dyn_L_hair_03_capsule
   center=<4.8194 0.7366 0.7063>
   rotation=<rot -8.1991 94.1047 -88.6979>
  }]
 dyn_L_hair_04 [{
   shapeType=Capsule
   length=10.2142
   radius=3.988
   shapeName=dyn_L_hair_04_capsule
   center=<8.084 2.5954 0.0675>
   rotation=<rot 0 90.2382 -86.829>
  }]
 dyn_M_front_hair_01 [{shapeType=Capsule, length=3.6359, radius=2.8765, shapeName=dyn_M_front_hair_01_capsule, center=<3.2777 1.1664 0>}]
 dyn_M_front_hair_02 [{shapeType=Capsule, length=4.9689, radius=2.3988, shapeName=dyn_M_front_hair_02_capsule, center=<3.8534 -0.0251 0>}]
 dyn_M_hair_01 [{
   shapeType=Capsule
   length=4.774
   radius=4.183
   shapeName=dyn_M_hair_01_capsule
   center=<5.8093 1.536 0.0807>
   rotation=<rot 89.2037 179.9999 -180>
  }]
 dyn_M_hair_02 [{
   shapeType=Capsule
   length=5.8039
   radius=4.3097
   shapeName=dyn_M_hair_02_capsule
   center=<2.5415 -1.4801 -0.1779>
   rotation=<rot 5.8701 97.2065 -84.8695>
  }]
 dyn_M_hair_03 [{
   shapeType=Capsule
   length=8.0256
   radius=3.7926
   shapeName=dyn_M_hair_03_capsule
   center=<4.1025 -0.001 0.1029>
   rotation=<rot 1.3163 99.8036 -86.6999>
  }]
 dyn_M_hair_04 [{
   shapeType=Capsule
   length=10.0866
   radius=3.9079
   shapeName=dyn_M_hair_04_capsule
   center=<7.5942 2.3075 1.2249>
   rotation=<rot -0.2916 94.1867 -87.3915>
  }]
 dyn_R_front_hair_01 [{shapeType=Capsule, length=4.3259, radius=2.5606, shapeName=dyn_R_front_hair_01_capsule, center=<2.8131 1.333 0>}]
 dyn_R_front_hair_02 [{shapeType=Capsule, length=3.7134, radius=3.105, shapeName=dyn_R_front_hair_02_capsule, center=<3.6511 1.7427 0>, rotation=<rot 0 -90 0>}]
 dyn_R_front_hair_03 [{
   shapeType=Capsule
   length=2.2885
   radius=3.7151
   shapeName=dyn_R_front_hair_03_capsule
   center=<5.6004 1.2027 -1.0653>
   rotation=<rot 0 75.173 3.3062>
  }]
 dyn_R_hair_01 [{
   shapeType=Capsule
   length=5.929
   radius=4.0735
   shapeName=dyn_R_hair_01_capsule
   center=<3.2159 0.1814 1.5382>
   rotation=<rot -13.5667 -70.4844 -80.2366>
  }]
 dyn_R_hair_02 [{
   shapeType=Capsule
   length=8.2544
   radius=3.4147
   shapeName=dyn_R_hair_02_capsule
   center=<3.7792 0.9311 -0.0888>
   rotation=<rot -60.0795 -116.9096 -65.3896>
  }]
 dyn_R_hair_03 [{
   shapeType=Capsule
   length=5.5036
   radius=3.3582
   shapeName=dyn_R_hair_03_capsule
   center=<2.8713 0.4027 -0.4814>
   rotation=<rot 23.0894 -83.4965 -80.2887>
  }]
 dyn_R_hair_04 [
  {
   shapeType=Capsule
   length=7.3501
   radius=3.7254
   shapeName=dyn_R_hair_04_capsule
   center=<9.4965 -2.3274 -0.9166>
   rotation=<rot 0 -107.003 -88.7494>
  }
  {
   shapeType=Capsule
   length=3.7428
   radius=1.3327
   shapeName=dyn_R_hair_04_capsule
   center=<8.0492 0.732 -4.1731>
   rotation=<rot -0.6255 -77.0085 -52.4054>
  }
  {
   shapeType=Capsule
   length=1.4619
   radius=0.7799
   shapeName=dyn_R_hair_04_capsule
   center=<11.5793 -0.01 3.0288>
   rotation=<rot 0 -107.0029 -106.3898>
  }
  {
   shapeType=Capsule
   length=2.0659
   radius=0.6546
   shapeName=dyn_R_hair_04_capsule
   center=<6.601 -4.3339 -5.0523>
   rotation=<rot 18.7412 -122.0394 -50.4909>
  }
 ]
 dyn_sleeve_01 [{
   shapeType=Capsule
   length=8.1985
   radius=7.6564
   shapeName=dyn_sleeve_01_capsule
   center=<5.1477 -0.0013 -5.8377>
   rotation=<rot 0.0082 -90.0056 93.6105>
  }]
 dyn_sleeve_02 [{
   shapeType=Capsule
   length=7.7301
   radius=8.1248
   shapeName=dyn_sleeve_02_capsule
   center=<3.3955 -4.0512 -2.4977>
   rotation=<rot 0.0081 -90.0061 98.6604>
  }]
 dyn_sleeve_03 [{
   shapeType=Capsule
   length=8.0179
   radius=7.8369
   shapeName=dyn_sleeve_03_capsule
   center=<3.1234 -1.6789 -4.3267>
   rotation=<rot 0.005 -89.9917 94.6801>
  }]
 dyn_sleeve_04 [{
   shapeType=Capsule
   length=8.1985
   radius=7.6564
   shapeName=dyn_sleeve_04_capsule
   center=<3.6916 6.8195 -0.5628>
   rotation=<rot 0.0069 -90.0065 98.6581>
  }]
 dyn_sweatshirt_L_back_01 [{
   shapeType=Capsule
   length=9.2024
   radius=6.6413
   shapeName=dyn_sweatshirt_L_back_01_capsule
   center=<8.074 0.8588 -6.8168>
   rotation=<rot -77.3986 90.0312 179.9995>
  }]
 dyn_sweatshirt_L_back_02 [{
   shapeType=Capsule
   length=9.2024
   radius=6.6413
   shapeName=dyn_sweatshirt_L_back_02_capsule
   center=<8.074 0.8588 -6.8168>
   rotation=<rot -77.3986 90.0312 179.9995>
  }]
 dyn_sweatshirt_M_back_01 [{
   shapeType=Capsule
   length=9.2024
   radius=6.6413
   shapeName=dyn_sweatshirt_M_back_01_capsule
   center=<8.0744 0.1074 -4.6998>
   rotation=<rot -88.6863 90.0358 179.9946>
  }]
 dyn_sweatshirt_M_back_02 [{
   shapeType=Capsule
   length=9.2024
   radius=6.6413
   shapeName=dyn_sweatshirt_M_back_02_capsule
   center=<8.0744 0.1074 -4.6998>
   rotation=<rot -88.6863 90.0358 179.9946>
  }]
 dyn_sweatshirt_R_back_01 [{
   shapeType=Capsule
   length=9.2024
   radius=6.6413
   shapeName=dyn_sweatshirt_R_back_01_capsule
   center=<8.0747 -0.3774 -5.0115>
   rotation=<rot -75.0604 -89.9704 0.0005>
  }]
 dyn_sweatshirt_R_back_02 [{
   shapeType=Capsule
   length=9.2024
   radius=6.6413
   shapeName=dyn_sweatshirt_R_back_02_capsule
   center=<8.0747 -0.3774 -5.0115>
   rotation=<rot -75.0604 -89.9704 0.0005>
  }]
 head [
  {shapeType=Capsule, length=6.5345, radius=6.5559, center=<3.1811 2.6041 0.3139>, rotation=<rot 0.5016 -90.415 -88.3249>}
  {
   shapeType=Capsule
   length=3.3888
   radius=6.0811
   shapeName=head_capsule
   center=<1.6133 -0.0375 -0.0239>
   rotation=<rot -89.1519 179.9981 -180>
  }
 ]
 spine_02 [{shapeType=Capsule, length=8.9683, radius=14.9365, shapeName=spine_02_capsule, rotation=<rot 0 -90 0>}]
 spine_04 [{shapeType=Capsule, length=17.3079, radius=11.924, shapeName=spine_04_capsule, center=<0 -0.7151 0>, rotation=<rot 0 -90 0>}]
 spine_05 [
  {shapeType=Capsule, length=23.992, radius=9.258, shapeName=spine_05_capsule, center=<0 -1.9176 0>, rotation=<rot 0 -90 0>}
  {shapeType=Capsule, length=23.9815, radius=9.2685, shapeName=spine_05_capsule, center=<-4.5244 1.1348 0>, rotation=<rot 0 -90 0>}
 ]
 upperarm_l [{
   shapeType=Capsule
   length=11.7097
   radius=6.143
   shapeName=upperarm_l_capsule
   center=<6.4261 -0.4979 -0.6218>
   rotation=<rot -0.0001 -90 -88.9267>
  }]
 upperarm_r [{
   shapeType=Capsule
   length=22.5029
   radius=6.5826
   shapeName=upperarm_l_capsule
   center=<-12.065 0.4979 1.2786>
   rotation=<rot -0.0001 -90 92.5487>
  }]
}
constraints [
 {
  bone1Name=head
  bone2Name=spine_05
  swing1LimitDegrees=45
  swing1Motion=Free
  swing2LimitDegrees=45
  swing2Motion=Free
  twistLimitDegrees=45
  twistMotion=Free
 }
 {
  bone1Name=spine_05
  bone2Name=spine_04
  swing1LimitDegrees=45
  swing1Motion=Free
  swing2LimitDegrees=45
  swing2Motion=Free
  twistLimitDegrees=45
  twistMotion=Free
 }
 {
  bone1Name=BASE_R_hair
  bone2Name=head
  swing1LimitDegrees=45
  swing1Motion=Free
  swing2LimitDegrees=45
  swing2Motion=Free
  twistLimitDegrees=45
  twistMotion=Free
 }
 {bone1Name=dyn_R_hair_01, bone2Name=BASE_R_hair, swing1LimitDegrees=40, swing2LimitDegrees=20, twistLimitDegrees=5}
 {bone1Name=dyn_R_hair_02, bone2Name=dyn_R_hair_01, swing1LimitDegrees=40, swing2LimitDegrees=10, twistLimitDegrees=5}
 {bone1Name=dyn_R_hair_03, bone2Name=dyn_R_hair_02, swing1LimitDegrees=40, swing2LimitDegrees=20, twistLimitDegrees=5}
 {bone1Name=dyn_R_hair_04, bone2Name=dyn_R_hair_03, swing1LimitDegrees=40, swing2LimitDegrees=10, twistLimitDegrees=5}
 {bone1Name=dyn_M_front_hair_01, bone2Name=BASE_M_front_hair, swing1LimitDegrees=15, swing2LimitDegrees=10, twistLimitDegrees=10}
 {bone1Name=dyn_M_front_hair_02, bone2Name=dyn_M_front_hair_01, swing1LimitDegrees=15, swing2LimitDegrees=10, twistLimitDegrees=10}
 {bone1Name=dyn_M_hair_01, bone2Name=BASE_M_hair, swing1LimitDegrees=30, swing2LimitDegrees=10, twistLimitDegrees=5}
 {bone1Name=dyn_M_hair_02, bone2Name=dyn_M_hair_01, swing1LimitDegrees=30, swing2LimitDegrees=10, twistLimitDegrees=5}
 {bone1Name=dyn_M_hair_03, bone2Name=dyn_M_hair_02, swing1LimitDegrees=30, swing2LimitDegrees=10, twistLimitDegrees=5}
 {bone1Name=dyn_M_hair_04, bone2Name=dyn_M_hair_03, swing1LimitDegrees=30, swing2LimitDegrees=10, twistLimitDegrees=5}
 {bone1Name=dyn_R_front_hair_01, bone2Name=BASE_R_front_hair, swing1LimitDegrees=20, swing2LimitDegrees=10, twistLimitDegrees=5}
 {bone1Name=dyn_R_front_hair_02, bone2Name=dyn_R_front_hair_01, swing1LimitDegrees=20, swing2LimitDegrees=10, twistLimitDegrees=5}
 {bone1Name=dyn_R_front_hair_03, bone2Name=dyn_R_front_hair_02, swing1LimitDegrees=20, swing2LimitDegrees=10, twistLimitDegrees=5}
 {bone1Name=dyn_L_hair_04, bone2Name=dyn_L_hair_03, swing1LimitDegrees=30, swing2LimitDegrees=10, twistLimitDegrees=5}
 {bone1Name=dyn_L_hair_03, bone2Name=dyn_L_hair_02, swing1LimitDegrees=30, swing2LimitDegrees=10, twistLimitDegrees=5}
 {bone1Name=dyn_L_hair_02, bone2Name=dyn_L_hair_01, swing1LimitDegrees=30, swing2LimitDegrees=10, twistLimitDegrees=5}
 {bone1Name=dyn_L_hair_01, bone2Name=BASE_L_hair, swing1LimitDegrees=30, swing2LimitDegrees=10, twistLimitDegrees=5}
 {
  bone1Name=BASE_L_hair
  bone2Name=head
  swing1LimitDegrees=45
  swing1Motion=Free
  swing2LimitDegrees=45
  swing2Motion=Free
  twistLimitDegrees=45
  twistMotion=Free
 }
 {
  bone1Name=BASE_M_front_hair
  bone2Name=head
  swing1LimitDegrees=45
  swing1Motion=Free
  swing2LimitDegrees=45
  swing2Motion=Free
  twistLimitDegrees=45
  twistMotion=Free
 }
 {
  bone1Name=BASE_M_hair
  bone2Name=head
  swing1LimitDegrees=45
  swing1Motion=Free
  swing2LimitDegrees=45
  swing2Motion=Free
  twistLimitDegrees=45
  twistMotion=Free
 }
 {
  bone1Name=BASE_R_front_hair
  bone2Name=head
  swing1LimitDegrees=45
  swing1Motion=Free
  swing2LimitDegrees=45
  swing2Motion=Free
  twistLimitDegrees=45
  twistMotion=Free
 }
 {
  bone1Name=BASE_L_front_hair
  bone2Name=head
  swing1LimitDegrees=45
  swing1Motion=Free
  swing2LimitDegrees=45
  swing2Motion=Free
  twistLimitDegrees=45
  twistMotion=Free
 }
 {bone1Name=L_front_hair_01, bone2Name=BASE_L_front_hair, swing1LimitDegrees=20, swing2LimitDegrees=10, twistLimitDegrees=5}
 {bone1Name=L_front_hair_02, bone2Name=L_front_hair_01, swing1LimitDegrees=20, swing2LimitDegrees=10, twistLimitDegrees=5}
 {bone1Name=L_front_hair_03, bone2Name=L_front_hair_02, swing1LimitDegrees=20, swing2LimitDegrees=10, twistLimitDegrees=5}
 {
  bone1Name=clavicle_l
  bone2Name=spine_05
  swing1LimitDegrees=45
  swing1Motion=Free
  swing2LimitDegrees=45
  swing2Motion=Free
  twistLimitDegrees=45
  twistMotion=Free
 }
 {
  bone1Name=upperarm_l
  bone2Name=clavicle_l
  swing1LimitDegrees=45
  swing1Motion=Free
  swing2LimitDegrees=45
  swing2Motion=Free
  twistLimitDegrees=45
  twistMotion=Free
 }
 {
  bone1Name=clavicle_r
  bone2Name=spine_05
  swing1LimitDegrees=45
  swing1Motion=Free
  swing2LimitDegrees=45
  swing2Motion=Free
  twistLimitDegrees=45
  twistMotion=Free
 }
 {
  bone1Name=upperarm_r
  bone2Name=clavicle_r
  swing1LimitDegrees=45
  swing1Motion=Free
  swing2LimitDegrees=45
  swing2Motion=Free
  twistLimitDegrees=45
  twistMotion=Free
 }
 {
  bone1Name=spine_04
  bone2Name=spine_02
  swing1LimitDegrees=45
  swing1Motion=Free
  swing2LimitDegrees=45
  swing2Motion=Free
  twistLimitDegrees=45
  twistMotion=Free
 }
 {
  bone1Name=dyn_sweatshirt_L_back_01
  bone2Name=spine_02
  swing1LimitDegrees=45
  swing1Motion=Free
  swing2LimitDegrees=45
  swing2Motion=Free
  twistLimitDegrees=45
  twistMotion=Free
 }
 {
  bone1Name=dyn_sweatshirt_L_back_02
  bone2Name=dyn_sweatshirt_L_back_01
  swing1LimitDegrees=45
  swing1Motion=Free
  swing2LimitDegrees=45
  swing2Motion=Free
  twistLimitDegrees=45
  twistMotion=Free
 }
 {
  bone1Name=dyn_sweatshirt_M_back_01
  bone2Name=spine_02
  swing1LimitDegrees=45
  swing1Motion=Free
  swing2LimitDegrees=45
  swing2Motion=Free
  twistLimitDegrees=45
  twistMotion=Free
 }
 {
  bone1Name=dyn_sweatshirt_M_back_02
  bone2Name=dyn_sweatshirt_M_back_01
  swing1LimitDegrees=45
  swing1Motion=Free
  swing2LimitDegrees=45
  swing2Motion=Free
  twistLimitDegrees=45
  twistMotion=Free
 }
 {
  bone1Name=dyn_sweatshirt_R_back_01
  bone2Name=spine_02
  swing1LimitDegrees=45
  swing1Motion=Free
  swing2LimitDegrees=45
  swing2Motion=Free
  twistLimitDegrees=45
  twistMotion=Free
 }
 {
  bone1Name=dyn_sweatshirt_R_back_02
  bone2Name=dyn_sweatshirt_R_back_01
  swing1LimitDegrees=45
  swing1Motion=Free
  swing2LimitDegrees=45
  swing2Motion=Free
  twistLimitDegrees=45
  twistMotion=Free
 }
 {
  bone1Name=dyn_sleeve_01
  bone2Name=upperarm_r
  swing1LimitDegrees=45
  swing1Motion=Free
  swing2LimitDegrees=45
  swing2Motion=Free
  twistLimitDegrees=45
  twistMotion=Free
 }
 {
  bone1Name=dyn_sleeve_02
  bone2Name=upperarm_r
  swing1LimitDegrees=45
  swing1Motion=Free
  swing2LimitDegrees=45
  swing2Motion=Free
  twistLimitDegrees=45
  twistMotion=Free
 }
 {
  bone1Name=dyn_sleeve_03
  bone2Name=upperarm_r
  swing1LimitDegrees=45
  swing1Motion=Free
  swing2LimitDegrees=45
  swing2Motion=Free
  twistLimitDegrees=45
  twistMotion=Free
 }
 {
  bone1Name=dyn_sleeve_04
  bone2Name=upperarm_r
  swing1LimitDegrees=45
  swing1Motion=Free
  swing2LimitDegrees=45
  swing2Motion=Free
  twistLimitDegrees=45
  twistMotion=Free
 }
]
