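# Animation Blueprint for the WormChalk male face accessory, simulating its hair physics on the player character.
# It blends the attached parent's pose with the body mesh's from the clavicles, then runs a pre-sim Control Rig, neck-space RigidBody and post-sim Control Rig.
# Rigs receive backbling/cape state, plus skydive floating pre-sim.
# Each update refreshes that state and derives a gravity override from per-movement-state wind and gravity settings via a deprecated helper.
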AnimBlueprint /BRCosmetics/Characters/Player/Male/Medium/Bodies/M_MED_WormChalk/Meshes/Parts/M_MED_WormChalk_FaceAcc_AnimBP
target_skeleton=&/M_MED_WormChalk_FaceAcc_Skeleton
graphs [
 AnimGraph {nodes [
   n1 = Get Body_Mesh()
   n2 = Control Rig!(
    Source=Layered blend per bone(BasePose=Copy Pose From Mesh() {bUseAttachedParent=true}, BlendPoses_0=Copy Pose From Mesh(SourceMeshComponent=@n1)) {layerSetup [{branchFilters [{blendDepth=1, boneName=clavicle_l}, {blendDepth=1, boneName=clavicle_r}]}]}
    Alpha
    skydive:bool=Get bIsSkydiveFloating()
    cape:bool=Get backpack_cape_mod() -> (backpack_cape_mod:bool)
   ) -> (Pose) {
    alphaBoolBlend {blendOption=Linear}
    controlRigAssetReference {blueprintRigClass=&/M_MED_WormChalk_FaceAcc_CtrlRig_presim.M_MED_WormChalk_FaceAcc_CtrlRig_presim_C}
   }
   n3 = RigidBody!(ComponentPose=Local To Component(LocalPose=@n2), bAlphaBoolEnabled:bool, OverrideWorldGravity:Vector="0,0,0", ExternalForce) -> (Pose) {
    alphaInputType=Bool
    bOverrideWorldGravity=true
    overridePhysicsAsset=&/M_MED_WormChalk_FaceAcc_Physics_hair
    simulationSpace=BaseBoneSpace
    alphaBoolBlend {blendOption=Linear}
    baseBoneRef {boneName=neck_01}
    componentAppliedLinearAccClamp=<5000 5000 5000>
    componentLinearAccScale=<0.3 0.3 0.3>
    componentLinearVelScale=<0.3 0.3 0.3>
    overrideWorldGravity {z=-980}
    simSpaceSettings {
     maxAngularAcceleration=500
     maxAngularVelocity=3
     maxLinearAcceleration=400
     maxLinearVelocity=250
     velocityScaleZ=0.2
     externalLinearDragV=<0.1 0.1>
    }
   }
   n4 = Control Rig!(Source=Component To Local(ComponentPose=@n3), Alpha, backpack_wearing:bool=Get backpack_cape_mod()) -> (Pose) {
    alphaBoolBlend {blendOption=Linear}
    controlRigAssetReference {blueprintRigClass=&/M_MED_WormChalk_FaceAcc_CtrlRig_postsim.M_MED_WormChalk_FaceAcc_CtrlRig_postsim_C}
   }
   Output Pose(Result=@n4) {layerGroup=None}
  ]}
 EventGraph {nodes [
   Event OnCharacterCustomizationLoaded()
   > n1 = Cast To PlayerPawn_Athena(Object=TryGetPawnOwner())
   > Set Body_Mesh(Body_Mesh=GetSkeletalMeshForPartType(self=@n1))
   Event BlueprintUpdateAnimation()
   > n2 = Sequence!(execute) -> (then_0, then_1, then_2:Exec)
   n3 = CheckForBackbling()
   > Set backpack_cape_mod(backpack_cape_mod:bool=OR Boolean(A=@n3, B=@n3.bIsBackblingACape)) -> (Output_Get:bool)
   > Set!(
    execute
    Gravity Override Params_Skydive_GoingUp_28_#1_WindFrequency_28_#2:Vector="1.5,7,8"
    Gravity Override Params_Skydive_GoingUp_28_#1_WindAmplitude_31_#3:Vector="200,-500,400"
    Gravity Override Params_Skydive_GoingUp_28_#1_WindOffset_32_#4:Vector="0,0,0"
    Gravity Override Params_Skydive_GoingUp_28_#1_GravityOverride_15_#5:Vector="0,-300,0"
    Gravity Override Params_Skydive_GoingUp_28_#1_JointName_27_#6:Name=root
    Gravity Override Params_Skydive_FloatingDown_29_#7_WindFrequency_28_#2:Vector="1.5,7,8"
    Gravity Override Params_Skydive_FloatingDown_29_#7_WindAmplitude_31_#3:Vector="200,-500,400"
    Gravity Override Params_Skydive_FloatingDown_29_#7_WindOffset_32_#4:Vector="0,-600,400"
    Gravity Override Params_Skydive_FloatingDown_29_#7_GravityOverride_15_#5:Vector="0,-700,-300"
    Gravity Override Params_Skydive_FloatingDown_29_#7_JointName_27_#6:Name=root
    Gravity Override Params_Skydive_DivingDown_30_#8:Gravity Override Setting Struct Structure=(WindFrequency_28_#2=(X=1.5,Y=12,Z=10),WindAmplitude_31_#3=(X=50,Y=300,Z=100),WindOffset_32_#4=(X=0,Y=0,Z=0),GravityOverride_15_#5=(X=-1500,Y=0,Z=0),JointName_27_#6="pelvis")
    Gravity Override Params_Skydive_Parachuting_31_#9:Gravity Override Setting Struct Structure=(WindFrequency_28_#2=(X=1.5,Y=7,Z=8),WindAmplitude_31_#3=(X=200,Y=-500,Z=400),WindOffset_32_#4=(X=0,Y=0,Z=0),GravityOverride_15_#5=(X=0,Y=-800,Z=-980),JointName_27_#6="root")
    Gravity Override Params_OnGround_Standing_32_#10:Gravity Override Setting Struct Structure=(WindFrequency_28_#2=(X=0,Y=0,Z=0),WindAmplitude_31_#3=(X=0,Y=0,Z=0),WindOffset_32_#4=(X=0,Y=0,Z=0),GravityOverride_15_#5=(X=0,Y=0,Z=-980),JointName_27_#6="root")
    Gravity Override Params_OnGround_Moving_33_#11:Gravity Override Setting Struct Structure=(WindFrequency_28_#2=(X=0,Y=0,Z=0),WindAmplitude_31_#3=(X=0,Y=0,Z=0),WindOffset_32_#4=(X=0,Y=0,Z=0),GravityOverride_15_#5=(X=0,Y=0,Z=-980),JointName_27_#6="root")
    Gravity Override Params_EmoteOrMelee_34_#12:Gravity Override Setting Struct Structure=(WindFrequency_28_#2=(X=0,Y=0,Z=0),WindAmplitude_31_#3=(X=0,Y=0,Z=0),WindOffset_32_#4=(X=0,Y=0,Z=0),GravityOverride_15_#5=(X=0,Y=0,Z=-980),JointName_27_#6="root")
    Gravity Override Params_Swimming_35_#13:Gravity Override Setting Struct Structure=(WindFrequency_28_#2=(X=0,Y=0,Z=0),WindAmplitude_31_#3=(X=0,Y=0,Z=0),WindOffset_32_#4=(X=0,Y=0,Z=0),GravityOverride_15_#5=(X=-800,Y=0,Z=0),JointName_27_#6="pelvis")
    Gravity Override Params_Front_End_36_#14:Gravity Override Setting Struct Structure=(WindFrequency_28_#2=(X=0,Y=0,Z=0),WindAmplitude_31_#3=(X=0,Y=0,Z=0),WindOffset_32_#4=(X=0,Y=0,Z=0),GravityOverride_15_#5=(X=0,Y=0,Z=-980),JointName_27_#6="root")
    Gravity Override Params_DBNO_40_#15:Gravity Override Setting Struct Structure=(WindFrequency_28_#2=(X=0,Y=0,Z=0),WindAmplitude_31_#3=(X=0,Y=0,Z=0),WindOffset_32_#4=(X=0,Y=0,Z=0),GravityOverride_15_#5=(X=-980,Y=400,Z=0),JointName_27_#6="pelvis")
    Gravity Override Params_DBNO_Carried_41_#16:Gravity Override Setting Struct Structure=(WindFrequency_28_#2=(X=0,Y=0,Z=0),WindAmplitude_31_#3=(X=0,Y=0,Z=0),WindOffset_32_#4=(X=0,Y=0,Z=0),GravityOverride_15_#5=(X=-980,Y=500,Z=-500),JointName_27_#6="pelvis")
    Gravity Override Params_OnGround_Falling_45_#17:Gravity Override Setting Struct Structure=(WindFrequency_28_#2=(X=0,Y=0,Z=0),WindAmplitude_31_#3=(X=0,Y=0,Z=0),WindOffset_32_#4=(X=0,Y=0,Z=0),GravityOverride_15_#5=(X=0,Y=0,Z=-980),JointName_27_#6="root")
    Gravity Override Params_Ziplining_46_#18:Gravity Override Setting Struct Structure=(WindFrequency_28_#2=(X=0,Y=0,Z=0),WindAmplitude_31_#3=(X=0,Y=0,Z=0),WindOffset_32_#4=(X=0,Y=0,Z=0),GravityOverride_15_#5=(X=0,Y=0,Z=-980),JointName_27_#6="root")
    Gravity Override Params_Skydive_Surfing_49_#19:Gravity Override Setting Struct Structure=(WindFrequency_28_#2=(X=1.5,Y=7,Z=8),WindAmplitude_31_#3=(X=200,Y=-500,Z=400),WindOffset_32_#4=(X=0,Y=0,Z=0),GravityOverride_15_#5=(X=500,Y=-300,Z=-980),JointName_27_#6="root")
    Gravity Override Params_Riding_Standing_54_#20:Gravity Override Setting Struct Structure=(WindFrequency_28_#2=(X=0,Y=0,Z=0),WindAmplitude_31_#3=(X=0,Y=0,Z=0),WindOffset_32_#4=(X=0,Y=0,Z=0),GravityOverride_15_#5=(X=0,Y=-350,Z=-980),JointName_27_#6="root")
    Gravity Override Params_Riding_Moving_55_#21:Gravity Override Setting Struct Structure=(WindFrequency_28_#2=(X=0,Y=0,Z=0),WindAmplitude_31_#3=(X=0,Y=0,Z=0),WindOffset_32_#4=(X=0,Y=0,Z=0),GravityOverride_15_#5=(X=0,Y=-450,Z=-450),JointName_27_#6="root")
    Gravity Override Params_Riding_Moving_Targeting_58_#22:Gravity Override Setting Struct Structure=(WindFrequency_28_#2=(X=0,Y=0,Z=0),WindAmplitude_31_#3=(X=0,Y=0,Z=0),WindOffset_32_#4=(X=0,Y=0,Z=0),GravityOverride_15_#5=(X=0,Y=-200,Z=-980),JointName_27_#6="root")
    Gravity Override Params_Tactical_Sprint_57_#23_WindFrequency_28_#2:Vector="0,0,0"
    Gravity Override Params_Tactical_Sprint_57_#23_WindAmplitude_31_#3:Vector="0,0,0"
    Gravity Override Params_Tactical_Sprint_57_#23_WindOffset_32_#4:Vector="0,0,0"
    Gravity Override Params_Tactical_Sprint_57_#23_GravityOverride_15_#5:Vector="0,-450,-1200"
    Gravity Override Params_Tactical_Sprint_57_#23_JointName_27_#6:Name=root
    Gravity Override Params_Rail_Grinding_61_#24:Gravity Override Setting Struct Structure=(WindFrequency_28_#2=(X=0,Y=0,Z=0),WindAmplitude_31_#3=(X=0,Y=0,Z=0),WindOffset_32_#4=(X=0,Y=0,Z=0),GravityOverride_15_#5=(X=0,Y=0,Z=-980),JointName_27_#6="root")
    Gravity Override Params_DBNO_Tumbling_67_#25:Gravity Override Setting Struct Structure=(WindFrequency_28_#2=(X=0,Y=0,Z=0),WindAmplitude_31_#3=(X=0,Y=0,Z=0),WindOffset_32_#4=(X=0,Y=0,Z=0),GravityOverride_15_#5=(X=-2000,Y=0,Z=0),JointName_27_#6="pelvis")
    Gravity Override Params_DBNO_RollFwd_68_#26:Gravity Override Setting Struct Structure=(WindFrequency_28_#2=(X=0,Y=0,Z=0),WindAmplitude_31_#3=(X=0,Y=0,Z=0),WindOffset_32_#4=(X=0,Y=0,Z=0),GravityOverride_15_#5=(X=-2000,Y=0,Z=0),JointName_27_#6="pelvis")
    Gravity Override Params_DBNO_RollBwd_69_#27:Gravity Override Setting Struct Structure=(WindFrequency_28_#2=(X=0,Y=0,Z=0),WindAmplitude_31_#3=(X=0,Y=0,Z=0),WindOffset_32_#4=(X=0,Y=0,Z=0),GravityOverride_15_#5=(X=-4000,Y=0,Z=0),JointName_27_#6="pelvis")
    Gravity Override Params_DBNO_Impulse_70_#28:Gravity Override Setting Struct Structure=(WindFrequency_28_#2=(X=0,Y=0,Z=0),WindAmplitude_31_#3=(X=0,Y=0,Z=0),WindOffset_32_#4=(X=0,Y=0,Z=0),GravityOverride_15_#5=(X=-2000,Y=0,Z=0),JointName_27_#6="pelvis")
   ) -> (then, Output_Get:Gravity Override Params Struct Structure)
   n4 = Gravity Override State Function DEPRECATED(
    execute=@n2.then_1
    CharacterPartAnimInstance=Self-Reference()
    GameTimeInSeconds=GetGameTimeInSeconds()
    GravityOverrideParams=Get Gravity Override Params()
   )
   > Set Gravity Override(Gravity Override=@n4)
  ]}
]
variables [
 backpack_cape_mod {type=boolean}
 Body_Mesh {type=SkeletalMeshComponent}
 Gravity Override {type=Vector}
 Gravity Override Params {
  type=GravityOverrideParamsStruct
  default {skydive_FloatingDown {gravityOverride=<0 -700 50>}, skydive_GoingUp {gravityOverride=<0 -300 100>}}
 }
]
